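; Blueprint function library of shared helpers for Pilgrim Quickplay UI, performer animation and song queries.
; Instrument icon lookup derives a Sparks instrument when unset, resolves an index and selects a material, with pro variants falling back to standard icons.
; Setup and reset functions link or unlink instrument animation layers on every playspace pawn through its instrument anim component.
; The key and mode query reads the Sparks music playspace, returning false with C Major without one.
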(fn GetIconMaterialByInstrument
  (params (Instrument :type EPilgrimTrackType%20Enum :default "TrackGuitar") (SongMetaData :type Pilgrim%20Song%20Metadata%20Object%20Reference) (SparksInstrument :type ESparksInstrumentType%20Enum) (bUseKeytarIcon :type Boolean :default true) (UseProIcons :type Boolean :default true))
  (return (UI|Instrument|GetInstrumentIconMaterialbyIndex (UI|Instrument|GetInstrumentIconIndex Instrument SongMetaData (select (Utilities|Enum|Equal(Enum) SparksInstrument "None") (Class|PilgrimCoreBPFL|PilgrimTrackTypetoSparksInstrument Instrument) SparksInstrument) bUseKeytarIcon (Variables|Get__WorldContext)) UseProIcons (Variables|Get__WorldContext))))

(fn GetInstrumentIconMaterialByIndex
  (params (Index :type Integer) (UseProIcons :type Boolean :default true))
  (return (Utilities|Select :Index Index :Option%200 "/PilgrimQuickplay/Widgets/Materials/MI_UI_Icon_Instrument_Type_Guitar.MI_UI_Icon_Instrument_Type_Guitar" :Option%201 "/PilgrimQuickplay/Widgets/Materials/MI_UI_Icon_Instrument_Type_Bass.MI_UI_Icon_Instrument_Type_Bass" :Option%202 "/PilgrimQuickplay/Widgets/Materials/MI_UI_Icon_Instrument_Type_Vocals.MI_UI_Icon_Instrument_Type_Vocals" :Option%203 "/PilgrimQuickplay/Widgets/Materials/MI_UI_Icon_Instrument_Type_Drums.MI_UI_Icon_Instrument_Type_Drums" :Option%204 (select UseProIcons "/PilgrimQuickplay/Widgets/Materials/MI_UI_Icon_Instrument_Type_PeriphericalGuitar.MI_UI_Icon_Instrument_Type_PeriphericalGuitar" "/PilgrimQuickplay/Widgets/Materials/MI_UI_Icon_Instrument_Type_Guitar.MI_UI_Icon_Instrument_Type_Guitar") :Option%205 (select UseProIcons "/PilgrimQuickplay/Widgets/Materials/MI_UI_Icon_Instrument_Type_PeripheralKeytar.MI_UI_Icon_Instrument_Type_PeripheralKeytar" "/PilgrimQuickplay/Widgets/Materials/MI_UI_Icon_Instrument_Type_Keytar.MI_UI_Icon_Instrument_Type_Keytar") :Option%206 "/PilgrimQuickplay/Widgets/Materials/MI_UI_Icon_Instrument_Type_Keytar.MI_UI_Icon_Instrument_Type_Keytar" :Option%207 (select UseProIcons "/PilgrimQuickplay/Widgets/Materials/MI_UI_Icon_Instrument_Type_PeripheralBass.MI_UI_Icon_Instrument_Type_PeripheralBass" "/PilgrimQuickplay/Widgets/Materials/MI_UI_Icon_Instrument_Type_Bass.MI_UI_Icon_Instrument_Type_Bass") :Option%208 (select UseProIcons "/PilgrimQuickplay/Widgets/Materials/MI_UI_Icon_Instrument_Type_Periphercal_ProDrums.MI_UI_Icon_Instrument_Type_Periphercal_ProDrums" "/PilgrimQuickplay/Widgets/Materials/MI_UI_Icon_Instrument_Type_Drums.MI_UI_Icon_Instrument_Type_Drums") :Option%209 (select UseProIcons "/PilgrimQuickplay/Widgets/Materials/MI_UI_Icon_Instrument_Type_Periphercal_ProSnare.MI_UI_Icon_Instrument_Type_Periphercal_ProSnare" "/PilgrimQuickplay/Widgets/Materials/MI_UI_Icon_Instrument_Type_Drums.MI_UI_Icon_Instrument_Type_Drums") :Option%2010 (select UseProIcons "/PilgrimQuickplay/Widgets/Materials/MI_UI_Icon_Instrument_Type_ProVocals.MI_UI_Icon_Instrument_Type_ProVocals" "/PilgrimQuickplay/Widgets/Materials/MI_UI_Icon_Instrument_Type_Vocals.MI_UI_Icon_Instrument_Type_Vocals"))))

(fn LinkInstrumentAnimLayer
  (params (Pawn :type Pawn%20Object%20Reference))
  (bind _returnvalue (Actor|GetComponentbyClass Pawn "/Script/PilgrimCoreRuntime.PawnComponent_PilgrimInstrumentAnim"))
  (Utilities|IsValid _returnvalue
    (:Is%20Valid
      (Class|SparksInstrumentAnimPawnComponent|LinkInstrumentAnimationLayer _returnvalue (Class|SparksInstrumentAnimPawnComponent|GetPlayerAnimations _returnvalue)))))

(fn ResetAllPerformerAnims
  (params (Playspace :type Pilgrim%20Quickplay%20Playspace%20Object%20Reference))
  (for _array_element (Class|SparksMusicPlayspace|GetAllPlayerPawns Playspace)
    (Default|UnlinkInstrumentAnimLayer _array_element)))

(fn SetupAllPerformerAnims
  (params (Playspace :type Pilgrim%20Quickplay%20Playspace%20Object%20Reference))
  (for _array_element (Class|SparksMusicPlayspace|GetAllPlayerPawns Playspace)
    (CallFunction|LinkInstrumentAnimLayer _array_element)))

(fn TryGetCurrentSongKeyAndMode
  (params (Actor :type Actor%20Object%20Reference))
  (bind _returnvalue (Class|SparksMusicPlayspace|GetSparksMusicPlayspaceOfClass "/Script/SparksMusicPlayspaceRuntime.SparksMusicPlayspace" 0 Actor))
  (bind (_outmusickey _outmusickeymode _returnvalue_2) (Class|SparksMusicPlayspace|TryGetMusicKeyandMode _returnvalue))
  (Utilities|IsValid _returnvalue
    (:Is%20Valid
      (return _returnvalue_2 _outmusickey _outmusickeymode))
    (:Is%20Not%20Valid
      (return false "C" "Major"))))

(fn UnlinkInstrumentAnimLayer
  (params (Pawn :type Pawn%20Object%20Reference))
  (bind _returnvalue (Actor|GetComponentbyClass Pawn "/Script/PilgrimCoreRuntime.PawnComponent_PilgrimInstrumentAnim"))
  (Utilities|IsValid _returnvalue
    (:Is%20Valid
      (Class|SparksInstrumentAnimPawnComponent|UnlinkInstrumentAnimationLayer _returnvalue))))
