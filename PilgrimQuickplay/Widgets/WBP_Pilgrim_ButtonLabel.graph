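# WBP_Pilgrim_ButtonLabel is a UserWidget that shows a centered text label, named as a button label for Pilgrim Quickplay.
# An Overlay holds a single CommonTextBlock styled with Fortnite's Body1 typography and a bold Burbank Small font, defaulting to "Label".
# Its SetLabel function takes a LabelText argument and writes it to the text block.
# The PreConstruct, Construct and Tick events are empty.
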WidgetBlueprint /PilgrimQuickplay/Widgets/WBP_Pilgrim_ButtonLabel
graphs [
 EventGraph {nodes [
   Event PreConstruct()
   Event Construct()
   Event Tick()
  ]}
 SetLabel {nodes [
   n1 = SetLabel!() -> (then, LabelText:Text) {is=title K2Node_FunctionEntry}
   > "SetText (Text)"(self=Get text() -> (text:Common Text Object Reference), InText=@n1)
  ]}
]
variables []
widgets {
 info {inheritedWidgetCount=0, parentClass=&UserWidget, rootWidgetClass=&Overlay, widgetCount=2}
 widgets [
  Overlay_75 {widgetClassPath=&Overlay, widget=Overlay_75}
  text {
   widgetClassPath=&CommonTextBlock
   bIsVariable=true
   parent=Overlay_75
   widget=text
   properties {
    lineHeightPercentage=1.29
    style=/FortUILibrary/Styles/Typography/_TextStyle_Body1._TextStyle_Body1_C
    text=Label
    font {
     bForceMonospaced=false
     bMaterialIsStencil=false
     fontMaterial=null
     fontObject=&BurbankSmall
     letterSpacing=0
     monospacedWidth=1
     size=18
     skewAmount=0
     typefaceFontName=Bold
     outlineSettings {
      bApplyOutlineToDropShadows=false
      bMiteredCorners=false
      bSeparateFillAlpha=false
      outlineMaterial=null
      outlineSize=0
      outlineColor=<rgba 0 0 0 1>
     }
    }
    shadowOffset=<0 0>
    strikeBrush {
     bIsDynamicallyLoaded=false
     drawAs=NoDrawType
     imageType=NoImage
     mirroring=NoMirror
     resourceName=None
     resourceObject=null
     tiling=NoTile
     imageSize=<32 32>
     margin=<ltrb 0 0 0 0>
     outlineSettings {
      bUseBrushTransparency=false
      roundingType=HalfHeightRadius
      width=0
      color {colorUseRule=UseColor_Specified, specifiedColor=<rgba 0 0 0 0>}
      cornerRadii=<0 0 0 0>
     }
     tintColor {colorUseRule=UseColor_Specified, specifiedColor=<rgba 1 1 1 1>}
     uVRegion {bIsValid=false, max=<0 0>, min=<0 0>}
    }
   }
   slot {class=OverlaySlot, properties {horizontalAlignment=HAlign_Center, verticalAlignment=VAlign_Center}}
  }
 ]
}
