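# Test panel widget for Pilgrim Quickplay cosmetics that lists the band intro director's spotlight montages and locks a spotlight index for testing.
# Refresh, run on construct, button click or director assignment, prints numbered guitar, bass, drum, mic and keytar entries and derives the maximum index from the longest list.
# Arrow buttons step the index, which SetLockIndex clamps, displays and passes to the director.
# A successful refresh resets the index to -1.
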WidgetBlueprint /PilgrimQuickplay/Utils/WBP_Pilgrim_Cosmetics_TestPanel
graphs [
 Clear Panel {nodes [
   Clear Panel() {is=title K2Node_FunctionEntry}
   > Set Text(
    Text=MakeLiteralText(Value="INVTEXT(\"Data not loaded.  Please refresh.\")")
    self=Get CommonTextBlock_0() -> (CommonTextBlock_0:Common Text Object Reference)
   )
   > Set MinValue(MinValue:int=-1) -> (Output_Get:int=0)
   > Set MaxValue(MaxValue:int=-1) -> (Output_Get:int=0)
   > Set CurrentValue(CurrentValue:int=-1) -> (Output_Get:int=0)
  ]}
 EventGraph {nodes [
   "On Clicked (btn_Refresh)"() {is=title K2Node_ComponentBoundEvent}
   > Refresh()
   n1 = "On Button Clicked (Button_ZeroDigitDown)"!() -> (OutputDelegate, then, Button:Common Button Legacy Object Reference) {is=title K2Node_ComponentBoundEvent}
   n2 = Get CurrentValue() -> (CurrentValue:int=0)
   n3 = Decrement Int(" "=[@n1.then, @n8.then], Value=@n2)
   > n4 = Set CurrentValue(CurrentValue=@n3) -> (Output_Get=0)
   SetLockIndex!(execute=[@n7.then, @n4.then], self, NewVal:int=Get CurrentValue() -> (CurrentValue=0)) -> (then)
   n5 = "On Button Clicked (Button_ZeroDigitUp)"!() -> (OutputDelegate, then, Button:Common Button Legacy Object Reference) {is=title K2Node_ComponentBoundEvent}
   n6 = Increment Int(" "=[@n5.then, @n9.then], Value=@n2)
   > n7 = Set CurrentValue(CurrentValue=@n6) -> (Output_Get=0)
   n8 = "On Clicked (Button_ZeroDigitDown)"!() -> (OutputDelegate, then, Button:Common Button Base Object Reference) {is=title K2Node_ComponentBoundEvent}
   n9 = "On Clicked (Button_ZeroDigitUp)"!() -> (OutputDelegate, then, Button:Common Button Base Object Reference) {is=title K2Node_ComponentBoundEvent}
   Event Construct()
   > SetLabel!(
    execute
    self:WBP Pilgrim Button Label Object Reference=Get lbl_Btn_Refresh() -> (lbl_Btn_Refresh:WBP Pilgrim Button Label Object Reference)
    LabelText:Text="INVTEXT(\"Refresh\")"
   ) -> (then)
   > Refresh()
   Event PreConstruct()
   Event Tick()
  ]}
 GetSpotlightEntries {nodes [
   n1 = GetSpotlightEntries!() -> (then, Spotlight Array:Array of Fort Montage Item Definition Base Object References, Text In:Text) {is=title K2Node_FunctionEntry}
   > n2 = Set Text!(execute, Text=@n1.Text In) -> (then, Output_Get)
   > Set Count(Count:int=0) -> (Output_Get:int=0)
   > n3 = For Each Loop(Array:Array of Fort Montage Item Definition Base Object References=@n1) -> (Array Element:Fort Montage Item Definition Base Object Reference)
   Return Node!(execute=@n3.Completed, Results:Text=Get Text!() -> (Text), LineCount:int=Get Count() -> (Count:int=0))
   n4 = GetPersistentName!(self:Fort Item Definition Object Reference=@n3) -> (ReturnValue:String)
   > Set Text(Text=Format Text!(Format="INVTEXT(\"{Prev}\\r\\n    {idx} : {Curr}\")", Prev:Text=@n2, idx:int=@n3.Array Index, Curr:Text="To Text (String)"(InString=@n4)) -> (Result))
   > Increment Int(Value=Get Count() -> (Count=0))
  ]}
 Refresh {nodes [
   Refresh() {is=title K2Node_FunctionEntry}
   > Clear Panel() {is=title K2Node_CallFunction}
   n1 = Get BandIntro_Director() -> (BandIntro_Director:BP Pilgrim Quick Play Band Intro Director Object Reference)
   > n2 = Is Valid(InputObject=@n1)
   Set Text!(
    execute=@n2.Is Not Valid
    Text=MakeLiteralText(Value="INVTEXT(\"BandIntro Director not found.\")")
    self:Text Object Reference=Get CommonTextBlock_0()
   ) -> (then, Output_Get)
   n3 = Sequence!(execute=@n2.Is Valid) -> (then_0, then_1, then_2:Exec, then_3:Exec, then_4:Exec, then_5:Exec)
   > n4 = Set TextBlock(TextBlock:Text=Format Text!(Format="INVTEXT(\"Guitar Spotlights:\")") -> (Result)) -> (Output_Get:Text)
   n5 = GetSpotlightEntries!(
    Spotlight Array:Array of Fort Montage Item Definition Base Object References=Get GuitarSpotlights!(self:Pilgrim Playspace Component Intro Director Object Reference=@n1) -> (GuitarSpotlights:Array of Fort Montage Item Definition Base Object References)
    Text In:Text=@n4
   ) -> (Results:Text, LineCount:int=0) {is=title K2Node_CallFunction}
   > Set TextBlock(TextBlock=@n5)
   > Set MaxValue(MaxValue="Max (Integer)"(A=@n5.LineCount, B=Get MaxValue() -> (MaxValue:int=0))) -> (Output_Get=0)
   n6 = Set TextBlock(
    execute=@n3.then_1
    TextBlock=Format Text!(Format="INVTEXT(\"{Text}\\r\\nBass Spotlights:\")", Text:Text=Get TextBlock() -> (TextBlock:Text)) -> (Result)
   )
   n7 = GetSpotlightEntries(
    Spotlight Array=Get BassSpotlights!(self:Pilgrim Playspace Component Intro Director Object Reference=@n1) -> (BassSpotlights:Array of Fort Montage Item Definition Base Object References)
    Text In=@n6
   ) -> (LineCount=0)
   > Set TextBlock(TextBlock=@n7)
   > Set MaxValue(MaxValue="Max (Integer)"(A=@n7.LineCount, B=Get MaxValue() -> (MaxValue=0))) -> (Output_Get=0)
   n8 = Set TextBlock(execute=@n3.then_2, TextBlock=Format Text(Format="INVTEXT(\"{Text}\\r\\nDrum Spotlights:\")", Text=Get TextBlock()))
   n9 = GetSpotlightEntries(
    Spotlight Array=Get DrumSpotlights!(self:Pilgrim Playspace Component Intro Director Object Reference=@n1) -> (DrumSpotlights:Array of Fort Montage Item Definition Base Object References)
    Text In=@n8
   ) -> (LineCount=0)
   > Set TextBlock(TextBlock=@n9)
   > Set MaxValue(MaxValue="Max (Integer)"(A=@n9.LineCount, B=Get MaxValue() -> (MaxValue=0))) -> (Output_Get=0)
   n10 = Set TextBlock(execute=@n3.then_3, TextBlock=Format Text(Format="INVTEXT(\"{Text}\\r\\nMic Spotlights:\")", Text=Get TextBlock()))
   n11 = GetSpotlightEntries(
    Spotlight Array=Get MicSpotlights!(self:Pilgrim Playspace Component Intro Director Object Reference=@n1) -> (MicSpotlights:Array of Fort Montage Item Definition Base Object References)
    Text In=@n10
   ) -> (LineCount=0)
   > Set TextBlock(TextBlock=@n11)
   > Set MaxValue(MaxValue="Max (Integer)"(A=@n11.LineCount, B=Get MaxValue() -> (MaxValue=0))) -> (Output_Get=0)
   n12 = Set TextBlock(execute=@n3.then_4, TextBlock=Format Text(Format="INVTEXT(\"{Text}\\r\\nKeytar Spotlights:\")", Text=Get TextBlock()))
   n13 = GetSpotlightEntries(
    Spotlight Array=Get KeytarSpotlights!(self:Pilgrim Playspace Component Intro Director Object Reference=@n1) -> (KeytarSpotlights:Array of Fort Montage Item Definition Base Object References)
    Text In=@n12
   ) -> (LineCount=0)
   > Set TextBlock(TextBlock=@n13)
   > Set MaxValue(MaxValue="Max (Integer)"(A=@n13.LineCount, B=Get MaxValue() -> (MaxValue=0))) -> (Output_Get=0)
   n14 = Sequence!(execute=@n3.then_5) -> (then_0, then_1)
   > Set Text(Text=Get TextBlock(), self=Get CommonTextBlock_0())
   > n15 = Decrement Int(Value=Get MaxValue() -> (MaxValue=0))
   > Set MaxValue(MaxValue=@n15) -> (Output_Get=0)
   SetLockIndex(execute=@n14.then_1, NewVal=-1)
  ]}
 Set BandIntro Director {nodes [
   n1 = Set BandIntro Director!() -> (then, BandIntro_Director:BP Pilgrim Quick Play Band Intro Director Object Reference) {is=title K2Node_FunctionEntry}
   > Set BandIntro_Director(BandIntro_Director:BP Pilgrim Quick Play Band Intro Director Object Reference=@n1) -> (Output_Get:BP Pilgrim Quick Play Band Intro Director Object Reference)
   > Refresh() {is=title K2Node_CallFunction}
  ]}
 SetLockIndex {nodes [
   n1 = SetLockIndex!() -> (then, NewVal:int) {is=title K2Node_FunctionEntry}
   > n2 = Set CurrentValue(CurrentValue="Clamp (Integer)"(Value=@n1, Min=Get MinValue() -> (MinValue:int=0), Max=Get MaxValue() -> (MaxValue=0)) -> (ReturnValue=0)) -> (Output_Get=0)
   > Set Text(Text="To Text (Integer)"(Value=@n2), self=Get txt_Index() -> (txt_Index:Common Text Object Reference))
   n3 = Get BandIntro_Director()
   > Is Valid(InputObject=@n3)
   > SetSpotlightIndexLock!(execute, self:BP Pilgrim Quick Play Band Intro Director Object Reference=@n3, Index:int=Get CurrentValue() -> (CurrentValue=0)) -> (then)
  ]}
]
variables [
 BandIntro_Director {type=BP_PilgrimQuickPlay_BandIntro_Director_C, default=null}
 CurrentValue {type=integer, default=-1}
 MaxValue {type=integer, default=-1}
 MinValue {type=integer, default=-1}
]
widgets {
 info {inheritedWidgetCount=0, parentClass=&UserWidget, rootWidgetClass=&CanvasPanel, widgetCount=15}
 widgets [
  CanvasPanel_126 {widgetClassPath=&CanvasPanel, widget=CanvasPanel_126}
  CanvasPanel_127 {
   widgetClassPath=&CanvasPanel
   parent=CanvasPanel_126
   widget=CanvasPanel_127
   slot {
    class=CanvasPanelSlot
    properties {layoutData {alignment=<0 0>, anchors {maximum=<1 1>, minimum=<1 1>}, offsets=<ltrb -652.2603 -1045.8938 627.5276 1012.973>}}
   }
  }
  img_Background {
   widgetClassPath=&Image
   bIsVariable=true
   parent=CanvasPanel_127
   widget=img_Background
   properties {brush {
     bIsDynamicallyLoaded=false
     drawAs=Image
     imageType=NoImage
     mirroring=NoMirror
     resourceName=None
     resourceObject=null
     tiling=NoTile
     imageSize=<32 32>
     margin=<ltrb 0 0 0 0>
     outlineSettings {
      bUseBrushTransparency=false
      roundingType=HalfHeightRadius
      width=0
      color {colorUseRule=UseColor_Specified, specifiedColor=<rgba 0 0 0 0>}
      cornerRadii=<0 0 0 0>
     }
     tintColor {colorUseRule=UseColor_Specified, specifiedColor=<rgba 0 0 0 0.2>}
     uVRegion {bIsValid=false, max=<0 0>, min=<0 0>}
    }}
   slot {
    class=CanvasPanelSlot
    properties {bAutoSize=true, layoutData {alignment=<0 0>, anchors {maximum=<1 1>, minimum=<0 0>}, offsets=<ltrb 0 0 0 1.001>}}
   }
  }
  VerticalBox_0 {
   widgetClassPath=&VerticalBox
   parent=CanvasPanel_127
   widget=VerticalBox_0
   slot {
    class=CanvasPanelSlot
    properties {layoutData {alignment=<0 0>, anchors {maximum=<0 0>, minimum=<0 0>}, offsets=<ltrb 64 27.9319 304.2042 120.0901>}}
   }
  }
  Text_VariantName {
   widgetClassPath=&CommonTextBlock
   parent=VerticalBox_0
   widget=Text_VariantName
   properties {
    text=Index to lock
    textTransformPolicy=ToUpper
    font {
     bForceMonospaced=false
     bMaterialIsStencil=false
     fontMaterial=null
     fontObject=&HeadingNow
     letterSpacing=0
     monospacedWidth=1
     size=18
     skewAmount=0
     typefaceFontName=75 Medium Italic
     outlineSettings {
      bApplyOutlineToDropShadows=false
      bMiteredCorners=false
      bSeparateFillAlpha=false
      outlineMaterial=null
      outlineSize=0
      outlineColor=<rgba 0 0 0 1>
     }
    }
    shadowOffset=<0 0>
   }
   slot {class=VerticalBoxSlot}
  }
  HorizontalBox_0 {widgetClassPath=&HorizontalBox, parent=VerticalBox_0, widget=HorizontalBox_0, slot {class=VerticalBoxSlot}}
  Button_ZeroDigitDown {
   widgetClassPath=/FortUILibrary/Foundation/Buttons/PanelButton
   bIsVariable=true
   parent=HorizontalBox_0
   widget=Button_ZeroDigitDown
   properties {
    inputActionHorizontalAlignment=HAlign_Right
    inputActionVerticalAlignment=VAlign_Center
    minHeight=88
    minWidth=88
    onButtonBaseClicked=()
    onButtonBaseDoubleClicked=()
    onButtonBaseHovered=()
    onButtonBaseUnhovered=()
    onSelectedChangedBase=()
    style=/FortUILibrary/Foundation/Buttons/ButtonStyles/ButtonStyle_CycleArrow_Left.ButtonStyle_CycleArrow_Left_C
    inputActionRenderTranslation=<22 0>
    inputActionRimBrushSize=<40 40>
   }
   slot {class=HorizontalBoxSlot}
  }
  VerticalBox_1 {widgetClassPath=&VerticalBox, parent=HorizontalBox_0, widget=VerticalBox_1, slot {class=HorizontalBoxSlot}}
  Image_1 {
   widgetClassPath=&Image
   parent=VerticalBox_1
   widget=Image_1
   properties {
    visibility=SelfHitTestInvisible
    brush {
     bIsDynamicallyLoaded=false
     drawAs=Image
     imageType=NoImage
     mirroring=NoMirror
     resourceName=None
     resourceObject=null
     tiling=NoTile
     imageSize=<16 4>
     margin=<ltrb 0 0 0 0>
     outlineSettings {
      bUseBrushTransparency=false
      roundingType=HalfHeightRadius
      width=0
      color {colorUseRule=UseColor_Specified, specifiedColor=<rgba 0 0 0 0>}
      cornerRadii=<0 0 0 0>
     }
     tintColor {colorUseRule=UseColor_Specified, specifiedColor=<rgba 1 1 1 1>}
     uVRegion {bIsValid=false, max=<0 0>, min=<0 0>}
    }
    colorAndOpacity=<rgba 1 1 1 0.2>
   }
   slot {class=VerticalBoxSlot}
  }
  txt_Index {
   widgetClassPath=&CommonTextBlock
   bIsVariable=true
   parent=VerticalBox_1
   widget=txt_Index
   properties {
    justification=Center
    scrollStyle=&TextScrollStyle-Base_C
    style=/FortUILibrary/Styles/Typography/Foundation/TextStyle-Base-XL-70pc.TextStyle-Base-XL-70pc_C
    text=-1
    colorAndOpacity {colorUseRule=UseColor_Specified, specifiedColor=<rgba 1 1 1 0.7>}
    font {
     bForceMonospaced=false
     bMaterialIsStencil=false
     fontMaterial=null
     fontObject=&NotoSans
     letterSpacing=0
     monospacedWidth=1
     size=44
     skewAmount=0
     typefaceFontName=NotoSans-Italic
     outlineSettings {
      bApplyOutlineToDropShadows=false
      bMiteredCorners=false
      bSeparateFillAlpha=false
      outlineMaterial=null
      outlineSize=0
      outlineColor=<rgba 0 0 0 1>
     }
    }
    shadowOffset=<0 0>
   }
   slot {
    class=VerticalBoxSlot
    properties {horizontalAlignment=HAlign_Center, verticalAlignment=VAlign_Center, padding=<ltrb 16 -16 16 -16>, size {sizeRule=Fill, value=1}}
   }
  }
  Image_0 {
   widgetClassPath=&Image
   parent=VerticalBox_1
   widget=Image_0
   properties {
    visibility=SelfHitTestInvisible
    brush {
     bIsDynamicallyLoaded=false
     drawAs=Image
     imageType=NoImage
     mirroring=NoMirror
     resourceName=None
     resourceObject=null
     tiling=NoTile
     imageSize=<16 4>
     margin=<ltrb 0 0 0 0>
     outlineSettings {
      bUseBrushTransparency=false
      roundingType=HalfHeightRadius
      width=0
      color {colorUseRule=UseColor_Specified, specifiedColor=<rgba 0 0 0 0>}
      cornerRadii=<0 0 0 0>
     }
     tintColor {colorUseRule=UseColor_Specified, specifiedColor=<rgba 1 1 1 1>}
     uVRegion {bIsValid=false, max=<0 0>, min=<0 0>}
    }
    colorAndOpacity=<rgba 1 1 1 0.2>
   }
   slot {class=VerticalBoxSlot}
  }
  Button_ZeroDigitUp {
   widgetClassPath=&PanelButton
   bIsVariable=true
   parent=HorizontalBox_0
   widget=Button_ZeroDigitUp
   properties {
    inputActionHorizontalAlignment=HAlign_Left
    inputActionVerticalAlignment=VAlign_Center
    minHeight=88
    minWidth=88
    onButtonBaseClicked=()
    onButtonBaseDoubleClicked=()
    onButtonBaseHovered=()
    onButtonBaseUnhovered=()
    onSelectedChangedBase=()
    style=/FortUILibrary/Foundation/Buttons/ButtonStyles/ButtonStyle_CycleArrow_Right.ButtonStyle_CycleArrow_Right_C
    inputActionRenderTranslation=<-22 0>
    inputActionRimBrushSize=<40 40>
   }
   slot {class=HorizontalBoxSlot, properties {verticalAlignment=VAlign_Center}}
  }
  btn_Refresh {
   widgetClassPath=&Button
   parent=CanvasPanel_127
   widget=btn_Refresh
   slot {
    class=CanvasPanelSlot
    properties {layoutData {alignment=<0 0>, anchors {maximum=<1 0>, minimum=<1 0>}, offsets=<ltrb -168.5887 12 161.0611 40>}}
   }
  }
  lbl_Btn_Refresh {
   widgetClassPath=/PilgrimQuickplay/Widgets/WBP_Pilgrim_ButtonLabel
   bIsVariable=true
   parent=btn_Refresh
   widget=lbl_Btn_Refresh
   properties {colorAndOpacity=<rgba 0 0 0 1>}
   slot {class=ButtonSlot}
  }
  CommonTextBlock_0 {
   widgetClassPath=&CommonTextBlock
   bIsVariable=true
   parent=CanvasPanel_127
   widget=CommonTextBlock_0
   properties {
    scrollStyle=&TextScrollStyle-Base_C
    style=/FortUILibrary/Styles/Typography/_TextStyle_TagExtraSmall._TextStyle_TagExtraSmall_C
    text=Text Block
    font {
     bForceMonospaced=false
     bMaterialIsStencil=false
     fontMaterial=null
     fontObject=&Burbank
     letterSpacing=0
     monospacedWidth=1
     size=12
     skewAmount=0
     typefaceFontName=Burbank-PlayerPower
     outlineSettings {
      bApplyOutlineToDropShadows=false
      bMiteredCorners=false
      bSeparateFillAlpha=false
      outlineMaterial=null
      outlineSize=0
      outlineColor=<rgba 0 0 0 1>
     }
    }
    shadowOffset=<0 0>
    strikeBrush {
     bIsDynamicallyLoaded=false
     drawAs=NoDrawType
     imageType=NoImage
     mirroring=NoMirror
     resourceName=None
     resourceObject=null
     tiling=NoTile
     imageSize=<32 32>
     margin=<ltrb 0 0 0 0>
     outlineSettings {
      bUseBrushTransparency=false
      roundingType=HalfHeightRadius
      width=0
      color {colorUseRule=UseColor_Specified, specifiedColor=<rgba 0 0 0 0>}
      cornerRadii=<0 0 0 0>
     }
     tintColor {colorUseRule=UseColor_Specified, specifiedColor=<rgba 1 1 1 1>}
     uVRegion {bIsValid=false, max=<0 0>, min=<0 0>}
    }
   }
   slot {
    class=CanvasPanelSlot
    properties {layoutData {alignment=<0 0>, anchors {maximum=<0 0>, minimum=<0 0>}, offsets=<ltrb 30.979 177.9219 567.4144 804.7608>}}
   }
  }
 ]
}
