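# Titled option rotator widget for the Sparks show manager UI that cycles through a text array, wrapping to the first entry.
# A hidden UIKit button drives hover and press animations, advances the option on click, then fires OnButtonClicked.
# Up to six pips, shown per option count, mark the current index by width and opacity through hard-coded branch chains.
# A gamepad-only action prompt appears while focused; functions also support index selection and default reset.
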WidgetBlueprint /PilgrimQuickplay/Utils/WBP_SparksShowManager_Rotator
animations [
 Hover {
  binding_count=2
  length_seconds=0.2
  object=&self:Hover
  sequence {bindings [
    SelectionBackingIndicator {
     id=#1
     possessed_class=Image
     tracks [{
       type=MovieSceneFloatTrack
       display_name=Render Opacity
       sections [{type=MovieSceneFloatSection, end_frame=4, start_frame=0, channels [None {type=MovieSceneScriptingFloatChannel, keys [[0,0],[4,1]]}]}]
      }]
    }
    Slots {
     id=#2
     possessed_class=Overlay
     tracks [{
       type=MovieScene2DTransformTrack
       sections [{
         type=MovieScene2DTransformSection
         channels [
          Scale.X {type=MovieSceneScriptingFloatChannel, keys [[0,1],[4,1.05]]}
          Scale.Y {type=MovieSceneScriptingFloatChannel, keys [[0,1],[4,1.05]]}
         ]
        }]
      }]
    }
   ]}
 }
 Pressed {
  binding_count=1
  length_seconds=0.2
  object=&self:Pressed
  sequence {bindings [GridPanel {
     id=#3
     possessed_class=Overlay
     tracks [{
       type=MovieScene2DTransformTrack
       sections [{
         type=MovieScene2DTransformSection
         channels [
          Scale.X {type=MovieSceneScriptingFloatChannel, keys [[0,1],[2,0.975],[4,1]]}
          Scale.Y {type=MovieSceneScriptingFloatChannel, keys [[0,1],[2,0.975],[4,1]]}
         ]
        }]
      }]
    }]}
 }
]
graphs [
 CycleToNextDot {nodes [
   n1 = CycleToNextDot!() -> (then, Index:int) {is=title K2Node_FunctionEntry}
   > Set LocalIndex(LocalIndex:int=@n1) -> (Output_Get:int=0)
   > n2 = Branch(Condition="Equal (Integer)"(A=Get LocalIndex() -> (LocalIndex:int=0)))
   > SetRenderOpacity(self=Get Image_Slot1() -> (Image_Slot1:Image Object Reference), InOpacity=1)
   > SetSize!(execute, self:Spacer Object Reference=Get Spacer_Slot1() -> (Spacer_Slot1:Spacer Object Reference), InSize:Vector 2D Structure=(X=48,Y=16)) -> (then)
   > SetRenderOpacity(
    self=[
     Get Image_Slot6() -> (Image_Slot6:Image Object Reference)
     Get Image_Slot2() -> (Image_Slot2:Image Object Reference)
     Get Image_Slot3() -> (Image_Slot3:Image Object Reference)
     Get Image_Slot4() -> (Image_Slot4:Image Object Reference)
     Get Image_Slot5() -> (Image_Slot5:Image Object Reference)
    ]
    InOpacity=0.5
   )
   > SetSize(
    self=[
     Get Spacer_Slot6() -> (Spacer_Slot6:Spacer Object Reference)
     Get Spacer_Slot2() -> (Spacer_Slot2:Spacer Object Reference)
     Get Spacer_Slot3() -> (Spacer_Slot3:Spacer Object Reference)
     Get Spacer_Slot4() -> (Spacer_Slot4:Spacer Object Reference)
     Get Spacer_Slot5() -> (Spacer_Slot5:Spacer Object Reference)
    ]
    InSize=(X=16,Y=16)
   )
   n3 = Branch(execute=@n2.else, Condition="Equal (Integer)"(A=Get LocalIndex() -> (LocalIndex=0), B=1))
   > SetRenderOpacity(self=Get Image_Slot2(), InOpacity=1)
   > SetSize(self=Get Spacer_Slot2(), InSize=(X=48,Y=16))
   > SetRenderOpacity(self=[Get Image_Slot1(), Get Image_Slot3(), Get Image_Slot4(), Get Image_Slot5(), Get Image_Slot6()], InOpacity=0.5)
   > SetSize(self=[Get Spacer_Slot6(), Get Spacer_Slot1(), Get Spacer_Slot3(), Get Spacer_Slot4(), Get Spacer_Slot5()], InSize=(X=16,Y=16))
   n4 = Branch(execute=@n3.else, Condition="Equal (Integer)"(A=Get LocalIndex() -> (LocalIndex=0), B=2))
   > SetRenderOpacity(self=Get Image_Slot3(), InOpacity=1)
   > SetSize(self=Get Spacer_Slot3(), InSize=(X=48,Y=16))
   > SetRenderOpacity(self=[Get Image_Slot1(), Get Image_Slot2(), Get Image_Slot4(), Get Image_Slot5(), Get Image_Slot6()], InOpacity=0.5)
   > SetSize(self=[Get Spacer_Slot6(), Get Spacer_Slot1(), Get Spacer_Slot2(), Get Spacer_Slot4(), Get Spacer_Slot5()], InSize=(X=16,Y=16))
   n5 = Branch(execute=@n4.else, Condition="Equal (Integer)"(A=Get LocalIndex() -> (LocalIndex=0), B=3))
   > SetRenderOpacity(self=Get Image_Slot4(), InOpacity=1)
   > SetSize(self=Get Spacer_Slot4(), InSize=(X=48,Y=16))
   > SetRenderOpacity(self=[Get Image_Slot1(), Get Image_Slot2(), Get Image_Slot3(), Get Image_Slot5(), Get Image_Slot6()], InOpacity=0.5)
   > SetSize(self=[Get Spacer_Slot1(), Get Spacer_Slot2(), Get Spacer_Slot3(), Get Spacer_Slot5(), Get Spacer_Slot6()], InSize=(X=16,Y=16))
   n6 = Branch(execute=@n5.else, Condition="Equal (Integer)"(A=Get LocalIndex() -> (LocalIndex=0), B=4))
   > SetRenderOpacity(self=Get Image_Slot5(), InOpacity=1)
   > SetSize(self=Get Spacer_Slot5(), InSize=(X=48,Y=16))
   > SetRenderOpacity(self=[Get Image_Slot1(), Get Image_Slot2(), Get Image_Slot3(), Get Image_Slot4(), Get Image_Slot6()], InOpacity=0.5)
   > SetSize(self=[Get Spacer_Slot1(), Get Spacer_Slot2(), Get Spacer_Slot3(), Get Spacer_Slot4(), Get Spacer_Slot6()], InSize=(X=16,Y=16))
   Branch(execute=@n6.else, Condition="Equal (Integer)"(A=Get LocalIndex() -> (LocalIndex=0), B=5))
   > SetRenderOpacity(self=Get Image_Slot6(), InOpacity=1)
   > SetSize(self=Get Spacer_Slot6(), InSize=(X=48,Y=16))
   > SetRenderOpacity(self=[Get Image_Slot1(), Get Image_Slot2(), Get Image_Slot3(), Get Image_Slot4(), Get Image_Slot5()], InOpacity=0.5)
   > SetSize(self=[Get Spacer_Slot2(), Get Spacer_Slot3(), Get Spacer_Slot4(), Get Spacer_Slot5(), Get Spacer_Slot1()], InSize=(X=16,Y=16))
  ]}
 CycleToNextOption {nodes [
   CycleToNextOption() {is=title K2Node_FunctionEntry}
   n1 = Get RotatorOptions() -> (RotatorOptions:Array of Texts)
   > n2 = Set MaxIndex(MaxIndex:int=int - int(A=Length(TargetArray:Array of Texts=@n1))) -> (Output_Get:int=0)
   > n3 = Branch(Condition="Equal (Integer)"(A=@n2, B=Get RotatorIndex() -> (RotatorIndex:int=0)))
   > n4 = Set RotatorIndex(RotatorIndex:int=0) -> (Output_Get:int=0)
   > "SetText (Text)"(
    self=Get Common_Text_CurrentName() -> (Common_Text_CurrentName:Common Text Object Reference)
    InText="Get (a copy)"(Array:Array of Texts=Get RotatorOptions(), Dimension 1=@n4) -> (Output:Text)
   )
   > CycleToNextDot!(execute, self, Index:int=@n4) -> (then) {is=title K2Node_CallFunction}
   n5 = Increment Int(" "=@n3.else, Value=Get RotatorIndex() -> (RotatorIndex=0))
   > "SetText (Text)"(self=Get Common_Text_CurrentName(), InText="Get (a copy)"(Array=Get RotatorOptions(), Dimension 1=@n5))
   > CycleToNextDot(Index=@n5)
  ]}
 EventGraph {nodes [
   "On Hovered (WBP_UIKit_Button_Regular)"!() -> (OutputDelegate, then, Button:Common Button Base Object Reference) {is=title K2Node_ComponentBoundEvent}
   > PlayAnimationForward(InAnimation=Get Hover() -> (Hover:Widget Animation Object Reference))
   > RequestTransition!(
    execute
    self:UIKit Block Object Reference=Get WBP_UIKit_Block_Outline() -> (WBP_UIKit_Block_Outline:WBP UIKit Block Outline Object Reference)
    State:EUIKitBlockVisualState Enum=Focused
   ) -> (then)
   n1 = Get CommonInputSubsystem()
   > n2 = Is Valid(InputObject=@n1)
   > n3 = SetVisibility(
    self=Get InputActionWidget() -> (InputActionWidget:Common Action Widget Object Reference)
    InVisibility=Select!(
     MouseAndKeyboard:ESlateVisibility Enum=Collapsed
     Gamepad:ESlateVisibility Enum=SelfHitTestInvisible
     Touch:ESlateVisibility Enum=Collapsed
     Count:ESlateVisibility Enum=Collapsed
     Index:ECommonInputType Enum=GetCurrentInputType(self=@n1)
    ) -> (ReturnValue)
   )
   Set bFocused(execute=[@n3.then, @n2.Is Not Valid], bFocused:bool=true) -> (Output_Get:bool)
   "On Unhovered (WBP_UIKit_Button_Regular)"!() -> (OutputDelegate, then, Button:Common Button Base Object Reference) {is=title K2Node_ComponentBoundEvent}
   > PlayAnimationReverse(InAnimation=Get Hover())
   > RequestTransition(self=Get WBP_UIKit_Block_Outline(), State=Unfocused)
   > SetVisibility(self=Get InputActionWidget(), InVisibility=Collapsed)
   > Set bFocused(bFocused=false)
   "On Clicked (WBP_UIKit_Button_Regular)"!() -> (OutputDelegate, then, Button:Common Button Base Object Reference) {is=title K2Node_ComponentBoundEvent}
   > PlayAnimation(InAnimation=Get Pressed() -> (Pressed:Widget Animation Object Reference))
   > CycleToNextOption() {is=title K2Node_CallFunction}
   > Call On Button Clicked!(execute:Exec, self:WBP Sparks Show Manager Rotator Object Reference) -> (then:Exec) {is=title K2Node_CallDelegate}
   n4 = ToggleInputAction() {is=title K2Node_CustomEvent}
   n5 = Get CommonInputSubsystem()
   > Is Valid(InputObject=@n5)
   > n6 = Switch on ECommonInputType!(execute, Selection:ECommonInputType Enum=GetCurrentInputType(self=@n5)) -> (MouseAndKeyboard:Exec, Gamepad:Exec, Touch:Exec, Count:Exec) {is=title K2Node_SwitchEnum}
   n7 = Branch(execute=@n6.Gamepad, Condition=Get bFocused() -> (bFocused:bool))
   > SetVisibility(self=Get InputActionWidget(), InVisibility=Visible)
   SetVisibility(execute=[@n6.Count, @n6.Touch, @n6.MouseAndKeyboard, @n7.else], self=Get InputActionWidget(), InVisibility=Collapsed)
   Event PreConstruct()
   > InitializeRotator()
   Event Construct()
   > ToggleInputAction() {is=title K2Node_CallFunction}
   > Bind Event to On Keybinds Changed(
    self:Fort Global UIContext Object Reference=Get FortGlobalUIContext() -> (ReturnValue:Fort Global UIContext Object Reference) {is=title K2Node_GetSubsystem}
    Delegate=@n4
   ) {is=title K2Node_AddDelegate}
  ]}
 InitializeRotator {nodes [
   InitializeRotator() {is=title K2Node_FunctionEntry}
   > "SetText (Text)"(self=Get Common_Text_CurrentName(), InText=Get CurrentText() -> (CurrentText:Text))
   > Set DefaultText(DefaultText:Text=Get CurrentText()) -> (Output_Get:Text)
   > SetWidthOverride(self=Get Size() -> (Size:Size Box Object Reference), InWidthOverride=Get WidgetWidth() -> (WidgetWidth:float))
   > WidgetSlotVisualization()
   > "SetText (Text)"(self=Get Text_Title() -> (Text_Title:Common Text Object Reference), InText=Get Title() -> (Title:Text))
  ]}
 OnButtonClicked {nodes [OnButtonClicked() {is=title K2Node_FunctionEntry}]}
 ResetToDefault {nodes [
   ResetToDefault() {is=title K2Node_FunctionEntry}
   > CycleToNextDot(Index=0)
   > "SetText (Text)"(self=Get Common_Text_CurrentName(), InText=Get DefaultText() -> (DefaultText:Text))
   > Set RotatorIndex(RotatorIndex=0) -> (Output_Get=0)
  ]}
 SetRotatorToSpecificIndex {nodes [
   n1 = SetRotatorToSpecificIndex!() -> (then, Index:int) {is=title K2Node_FunctionEntry}
   > n2 = Set RotatorIndex(RotatorIndex=@n1) -> (Output_Get=0)
   > "SetText (Text)"(self=Get Common_Text_CurrentName(), InText="Get (a copy)"(Array=Get RotatorOptions(), Dimension 1=@n2))
   > CycleToNextDot(Index=@n2)
  ]}
 WidgetSlotVisualization {nodes [
   WidgetSlotVisualization() {is=title K2Node_FunctionEntry}
   > n1 = Set OptionsDesired(OptionsDesired:int=Length(TargetArray=Get RotatorOptions())) -> (Output_Get:int=0)
   > n2 = Branch(Condition="Equal (Integer)"(A=@n1))
   > SetVisibility(
    self=[
     Get Slot4() -> (Slot4:Overlay Object Reference)
     Get Slot5() -> (Slot5:Overlay Object Reference)
     Get Slot6() -> (Slot6:Overlay Object Reference)
     Get Slot1() -> (Slot1:Overlay Object Reference)
     Get Slot2() -> (Slot2:Overlay Object Reference)
     Get Slot3() -> (Slot3:Overlay Object Reference)
    ]
    InVisibility=Collapsed
   )
   n3 = Branch(execute=@n2.else, Condition="Equal (Integer)"(A=Get OptionsDesired() -> (OptionsDesired:int=0), B=1))
   > SetVisibility(self=Get Slot1(), InVisibility=Visible)
   > SetVisibility(self=[Get Slot2(), Get Slot3(), Get Slot4(), Get Slot5(), Get Slot6()], InVisibility=Collapsed)
   n4 = Branch(execute=@n3.else, Condition="Equal (Integer)"(A=Get OptionsDesired() -> (OptionsDesired=0), B=2))
   > SetVisibility(self=[Get Slot1(), Get Slot2()], InVisibility=Visible)
   > SetVisibility(self=[Get Slot3(), Get Slot4(), Get Slot5(), Get Slot6()], InVisibility=Collapsed)
   n5 = Branch(execute=@n4.else, Condition="Equal (Integer)"(A=Get OptionsDesired() -> (OptionsDesired=0), B=3))
   > SetVisibility(self=[Get Slot1(), Get Slot2(), Get Slot3()], InVisibility=Visible)
   > SetVisibility(self=[Get Slot4(), Get Slot5(), Get Slot6()], InVisibility=Collapsed)
   n6 = Branch(execute=@n5.else, Condition="Equal (Integer)"(A=Get OptionsDesired() -> (OptionsDesired=0), B=4))
   > SetVisibility(self=[Get Slot1(), Get Slot2(), Get Slot3(), Get Slot4()], InVisibility=Visible)
   > SetVisibility(self=[Get Slot5(), Get Slot6()], InVisibility=Collapsed)
   n7 = Branch(execute=@n6.else, Condition="Equal (Integer)"(A=Get OptionsDesired() -> (OptionsDesired=0), B=5))
   > SetVisibility(self=[Get Slot1(), Get Slot2(), Get Slot3(), Get Slot4(), Get Slot5()], InVisibility=Visible)
   > SetVisibility(self=Get Slot6(), InVisibility=Collapsed)
   Branch(execute=@n7.else, Condition="Equal (Integer)"(A=Get OptionsDesired() -> (OptionsDesired=0), B=6))
   > SetVisibility(self=[Get Slot1(), Get Slot2(), Get Slot3(), Get Slot4(), Get Slot6(), Get Slot5()], InVisibility=Visible)
  ]}
]
variables [
 bFocused {type=boolean}
 CurrentText {type=string, default=Text}
 DefaultText {type=string, default=Text}
 OnButtonClicked {}
 RotatorIndex {type=integer}
 RotatorOptions {type="array<string>", default []}
 Title {type=string, default=Title}
 WidgetWidth {type=number, default=375}
]
widgets {
 info {inheritedWidgetCount=0, parentClass=&UserWidget, rootWidgetClass=&HorizontalBox, widgetCount=39}
 widgets [
  HorizontalBox_1 {widgetClassPath=&HorizontalBox, widget=HorizontalBox_1}
  Overlay_121 {widgetClassPath=&Overlay, parent=HorizontalBox_1, widget=Overlay_121, slot {class=HorizontalBoxSlot}}
  Text_Title {
   widgetClassPath=&CommonTextBlock
   bIsVariable=true
   parent=Overlay_121
   widget=Text_Title
   properties {
    justification=Center
    text=Title
    textTransformPolicy=ToUpper
    visibility=HitTestInvisible
    colorAndOpacity {colorUseRule=UseColor_Specified, specifiedColor=<rgba 0.5 0.8333 1 1>}
    font {
     bForceMonospaced=false
     bMaterialIsStencil=false
     fontMaterial=null
     fontObject=&HeadingNow
     letterSpacing=0
     monospacedWidth=1
     size=13.5
     skewAmount=0
     typefaceFontName=76 Bold Italic
     outlineSettings {
      bApplyOutlineToDropShadows=false
      bMiteredCorners=false
      bSeparateFillAlpha=false
      outlineMaterial=null
      outlineSize=0
      outlineColor=<rgba 0 0 0 1>
     }
    }
    shadowOffset=<0 0>
   }
   slot {class=OverlaySlot, properties {verticalAlignment=VAlign_Center, padding=<ltrb 5 0 0 0>}}
  }
  TitleSpacing {
   widgetClassPath=&SizeBox
   parent=Overlay_121
   widget=TitleSpacing
   properties {bOverride_WidthOverride=true, widthOverride=225}
   slot {class=OverlaySlot, properties {horizontalAlignment=HAlign_Fill, verticalAlignment=VAlign_Fill}}
  }
  GridPanel {widgetClassPath=&Overlay, bIsVariable=true, parent=HorizontalBox_1, widget=GridPanel, slot {class=HorizontalBoxSlot}}
  Overlay_173 {widgetClassPath=&Overlay, parent=GridPanel, widget=Overlay_173, slot {class=OverlaySlot}}
  Size {
   widgetClassPath=&SizeBox
   bIsVariable=true
   parent=Overlay_173
   widget=Size
   properties {bOverride_WidthOverride=true, widthOverride=375}
   slot {class=OverlaySlot, properties {horizontalAlignment=HAlign_Fill, verticalAlignment=VAlign_Fill}}
  }
  Backing {
   widgetClassPath=&Overlay
   bIsVariable=true
   parent=Overlay_173
   widget=Backing
   slot {class=OverlaySlot, properties {horizontalAlignment=HAlign_Fill, verticalAlignment=VAlign_Fill}}
  }
  Spacer_161 {
   widgetClassPath=&Spacer
   parent=Backing
   widget=Spacer_161
   properties {size=<1 48>}
   slot {class=OverlaySlot, properties {verticalAlignment=VAlign_Fill}}
  }
  SelectionBackingIndicator {
   widgetClassPath=&Image
   bIsVariable=true
   parent=Backing
   widget=SelectionBackingIndicator
   properties {
    renderOpacity=0
    visibility=SelfHitTestInvisible
    brush {
     bIsDynamicallyLoaded=false
     drawAs=Image
     imageType=NoImage
     mirroring=NoMirror
     resourceName=None
     resourceObject=/SparksUICommon/UILibrary/MI_FMCore_ButtonBacking_DifficultySelector
     tiling=NoTile
     imageSize=<32 32>
     margin=<ltrb 0 0 0 0>
     outlineSettings {
      bUseBrushTransparency=false
      roundingType=HalfHeightRadius
      width=0
      color {colorUseRule=UseColor_Specified, specifiedColor=<rgba 0 0 0 0>}
      cornerRadii=<0 0 0 0>
     }
     tintColor {colorUseRule=UseColor_Specified, specifiedColor=<rgba 1 1 1 1>}
     uVRegion {bIsValid=false, max=<0 0>, min=<0 0>}
    }
   }
   slot {class=OverlaySlot, properties {horizontalAlignment=HAlign_Fill, verticalAlignment=VAlign_Fill, padding=<ltrb 1 1 1 1>}}
  }
  WBP_UIKit_Block_Outline {
   widgetClassPath=/FortUILibrary/UIKit/Buttons/Blocks/WBP_UIKit_Block_Outline
   bIsVariable=true
   parent=Backing
   widget=WBP_UIKit_Block_Outline
   slot {class=OverlaySlot, properties {horizontalAlignment=HAlign_Fill, verticalAlignment=VAlign_Fill}}
  }
  Image_Background {
   widgetClassPath=&Border
   parent=Backing
   widget=Image_Background
   properties {
    background {
     bIsDynamicallyLoaded=false
     drawAs=RoundedBox
     imageType=NoImage
     mirroring=NoMirror
     resourceName=None
     resourceObject=null
     tiling=NoTile
     imageSize=<32 32>
     margin=<ltrb 0 0 0 0>
     outlineSettings {
      bUseBrushTransparency=false
      roundingType=FixedRadius
      width=0
      color {colorUseRule=UseColor_Specified, specifiedColor=<rgba 0 0 0 0>}
      cornerRadii=<14 14 14 14>
     }
     tintColor {colorUseRule=UseColor_Specified, specifiedColor=<rgba 0.0086 0.0097 0.0865 0.82>}
     uVRegion {bIsValid=false, max=<0 0>, min=<0 0>}
    }
    padding=<ltrb 14 14 14 14>
   }
   slot {class=OverlaySlot, properties {horizontalAlignment=HAlign_Fill, verticalAlignment=VAlign_Fill}}
  }
  WBP_UIKit_Button_Regular {
   widgetClassPath=&WBP_UIKit_Button_Regular
   bIsVariable=true
   parent=Backing
   widget=WBP_UIKit_Button_Regular
   properties {
    bHideInputActionWithKeyboard=true
    bUseNaturalNavigation=true
    buttonSize=Small
    buttonType=Quiet
    inputModeOverride=All
    renderOpacity=0
    text=""
    visibility=Visible
   }
   slot {class=OverlaySlot, properties {horizontalAlignment=HAlign_Fill, verticalAlignment=VAlign_Fill}}
  }
  GridPanel_0 {
   widgetClassPath=&GridPanel
   parent=Overlay_173
   widget=GridPanel_0
   properties {columnFill [1], rowFill [1]}
   slot {class=OverlaySlot, properties {horizontalAlignment=HAlign_Fill, verticalAlignment=VAlign_Center}}
  }
  HorizontalBox_0 {
   widgetClassPath=&HorizontalBox
   parent=GridPanel_0
   widget=HorizontalBox_0
   slot {class=GridSlot, properties {verticalAlignment=VAlign_Center, padding=<ltrb 16 0 8 0>}}
  }
  ScaleBox_0 {
   widgetClassPath=&ScaleBox
   parent=HorizontalBox_0
   widget=ScaleBox_0
   properties {stretch=UserSpecified, userSpecifiedScale=0.5}
   slot {class=HorizontalBoxSlot}
  }
  InputActionWidget {
   widgetClassPath=&CommonActionWidget
   bIsVariable=true
   parent=HorizontalBox_0
   widget=InputActionWidget
   properties {inputActions [{dataTable=&InputActionDataTable, rowName=Select}]}
   slot {class=HorizontalBoxSlot}
  }
  Common_Text_CurrentName {
   widgetClassPath=&CommonTextBlock
   bIsVariable=true
   parent=HorizontalBox_0
   widget=Common_Text_CurrentName
   properties {
    text=Text
    textTransformPolicy=ToUpper
    visibility=HitTestInvisible
    colorAndOpacity {colorUseRule=UseColor_Specified, specifiedColor=<rgba 0.5 0.8333 1 1>}
    font {
     bForceMonospaced=false
     bMaterialIsStencil=false
     fontMaterial=null
     fontObject=&HeadingNow
     letterSpacing=0
     monospacedWidth=1
     size=13.5
     skewAmount=0
     typefaceFontName=76 Bold Italic
     outlineSettings {
      bApplyOutlineToDropShadows=false
      bMiteredCorners=false
      bSeparateFillAlpha=false
      outlineMaterial=null
      outlineSize=0
      outlineColor=<rgba 0 0 0 1>
     }
    }
    shadowOffset=<0 0>
   }
   slot {class=HorizontalBoxSlot, properties {horizontalAlignment=HAlign_Left, verticalAlignment=VAlign_Center}}
  }
  Slots {
   widgetClassPath=&Overlay
   parent=HorizontalBox_0
   widget=Slots
   properties {visibility=HitTestInvisible, renderTransformPivot=<1 0.5>}
   slot {class=HorizontalBoxSlot, properties {horizontalAlignment=HAlign_Right, verticalAlignment=VAlign_Center, size {sizeRule=Fill, value=1}}}
  }
  PipBacking {
   widgetClassPath=&Image
   parent=Slots
   widget=PipBacking
   properties {
    renderOpacity=0.5
    visibility=Collapsed
    brush {
     bIsDynamicallyLoaded=false
     drawAs=Box
     imageType=FullColor
     mirroring=NoMirror
     resourceName=None
     resourceObject=&/M_UI_Pilgrim_RotatorIntensity
     tiling=NoTile
     imageSize=<24 24>
     margin=<ltrb 0.5 0.5 0.5 0.5>
     outlineSettings {
      bUseBrushTransparency=false
      roundingType=HalfHeightRadius
      width=0
      color {colorUseRule=UseColor_Specified, specifiedColor=<rgba 0 0 0 0>}
      cornerRadii=<0 0 0 0>
     }
     tintColor {colorUseRule=UseColor_Specified, specifiedColor=<rgba 1 1 1 1>}
     uVRegion {bIsValid=false, max=<0 0>, min=<0 0>}
    }
    colorAndOpacity=<rgba 0.0033 0.003 0.0497 1>
   }
   slot {class=OverlaySlot, properties {horizontalAlignment=HAlign_Fill, verticalAlignment=VAlign_Center}}
  }
  HorizontalBox_116 {
   widgetClassPath=&HorizontalBox
   parent=Slots
   widget=HorizontalBox_116
   slot {class=OverlaySlot, properties {horizontalAlignment=HAlign_Center, verticalAlignment=VAlign_Center, padding=<ltrb 4 0 4 0>}}
  }
  Slot1 {
   widgetClassPath=&Overlay
   bIsVariable=true
   parent=HorizontalBox_116
   widget=Slot1
   slot {class=HorizontalBoxSlot, properties {verticalAlignment=VAlign_Bottom}}
  }
  Spacer_Slot1 {
   widgetClassPath=&Spacer
   bIsVariable=true
   parent=Slot1
   widget=Spacer_Slot1
   properties {size=<40 1>}
   slot {class=OverlaySlot, properties {horizontalAlignment=HAlign_Fill, verticalAlignment=VAlign_Fill}}
  }
  Image_Slot1 {
   widgetClassPath=&Image
   bIsVariable=true
   parent=Slot1
   widget=Image_Slot1
   properties {
    visibility=SelfHitTestInvisible
    brush {
     bIsDynamicallyLoaded=false
     drawAs=Box
     imageType=FullColor
     mirroring=NoMirror
     resourceName=None
     resourceObject=&M_UI_Pilgrim_RotatorIntensity
     tiling=NoTile
     imageSize=<16 16>
     margin=<ltrb 0.5 0.5 0.5 0.5>
     outlineSettings {
      bUseBrushTransparency=false
      roundingType=HalfHeightRadius
      width=0
      color {colorUseRule=UseColor_Specified, specifiedColor=<rgba 0 0 0 0>}
      cornerRadii=<0 0 0 0>
     }
     tintColor {colorUseRule=UseColor_Specified, specifiedColor=<rgba 1 1 1 1>}
     uVRegion {bIsValid=false, max=<0 0>, min=<0 0>}
    }
    colorAndOpacity=<rgba 0.5 0.8333 1 1>
   }
   slot {class=OverlaySlot, properties {horizontalAlignment=HAlign_Fill, verticalAlignment=VAlign_Fill}}
  }
  Slot2 {
   widgetClassPath=&Overlay
   bIsVariable=true
   parent=HorizontalBox_116
   widget=Slot2
   slot {class=HorizontalBoxSlot, properties {verticalAlignment=VAlign_Bottom, padding=<ltrb 4 0 0 0>}}
  }
  Spacer_Slot2 {
   widgetClassPath=&Spacer
   bIsVariable=true
   parent=Slot2
   widget=Spacer_Slot2
   slot {class=OverlaySlot, properties {horizontalAlignment=HAlign_Fill, verticalAlignment=VAlign_Fill}}
  }
  Image_Slot2 {
   widgetClassPath=&Image
   bIsVariable=true
   parent=Slot2
   widget=Image_Slot2
   properties {
    renderOpacity=0.5
    visibility=SelfHitTestInvisible
    brush {
     bIsDynamicallyLoaded=false
     drawAs=Box
     imageType=FullColor
     mirroring=NoMirror
     resourceName=None
     resourceObject=&M_UI_Pilgrim_RotatorIntensity
     tiling=NoTile
     imageSize=<16 16>
     margin=<ltrb 0.5 0.5 0.5 0.5>
     outlineSettings {
      bUseBrushTransparency=false
      roundingType=HalfHeightRadius
      width=0
      color {colorUseRule=UseColor_Specified, specifiedColor=<rgba 0 0 0 0>}
      cornerRadii=<0 0 0 0>
     }
     tintColor {colorUseRule=UseColor_Specified, specifiedColor=<rgba 1 1 1 1>}
     uVRegion {bIsValid=false, max=<0 0>, min=<0 0>}
    }
    colorAndOpacity=<rgba 0.5 0.8333 1 1>
   }
   slot {class=OverlaySlot, properties {horizontalAlignment=HAlign_Fill, verticalAlignment=VAlign_Fill}}
  }
  Slot3 {
   widgetClassPath=&Overlay
   bIsVariable=true
   parent=HorizontalBox_116
   widget=Slot3
   slot {class=HorizontalBoxSlot, properties {verticalAlignment=VAlign_Bottom, padding=<ltrb 4 0 0 0>}}
  }
  Spacer_Slot3 {
   widgetClassPath=&Spacer
   bIsVariable=true
   parent=Slot3
   widget=Spacer_Slot3
   slot {class=OverlaySlot, properties {horizontalAlignment=HAlign_Fill, verticalAlignment=VAlign_Fill}}
  }
  Image_Slot3 {
   widgetClassPath=&Image
   bIsVariable=true
   parent=Slot3
   widget=Image_Slot3
   properties {
    renderOpacity=0.5
    visibility=SelfHitTestInvisible
    brush {
     bIsDynamicallyLoaded=false
     drawAs=Box
     imageType=FullColor
     mirroring=NoMirror
     resourceName=None
     resourceObject=&M_UI_Pilgrim_RotatorIntensity
     tiling=NoTile
     imageSize=<16 16>
     margin=<ltrb 0.5 0.5 0.5 0.5>
     outlineSettings {
      bUseBrushTransparency=false
      roundingType=HalfHeightRadius
      width=0
      color {colorUseRule=UseColor_Specified, specifiedColor=<rgba 0 0 0 0>}
      cornerRadii=<0 0 0 0>
     }
     tintColor {colorUseRule=UseColor_Specified, specifiedColor=<rgba 1 1 1 1>}
     uVRegion {bIsValid=false, max=<0 0>, min=<0 0>}
    }
    colorAndOpacity=<rgba 0.5 0.8333 1 1>
   }
   slot {class=OverlaySlot, properties {horizontalAlignment=HAlign_Fill, verticalAlignment=VAlign_Fill}}
  }
  Slot4 {
   widgetClassPath=&Overlay
   bIsVariable=true
   parent=HorizontalBox_116
   widget=Slot4
   slot {class=HorizontalBoxSlot, properties {verticalAlignment=VAlign_Bottom, padding=<ltrb 4 0 0 0>}}
  }
  Spacer_Slot4 {
   widgetClassPath=&Spacer
   bIsVariable=true
   parent=Slot4
   widget=Spacer_Slot4
   slot {class=OverlaySlot, properties {horizontalAlignment=HAlign_Fill, verticalAlignment=VAlign_Fill}}
  }
  Image_Slot4 {
   widgetClassPath=&Image
   bIsVariable=true
   parent=Slot4
   widget=Image_Slot4
   properties {
    renderOpacity=0.5
    visibility=SelfHitTestInvisible
    brush {
     bIsDynamicallyLoaded=false
     drawAs=Box
     imageType=FullColor
     mirroring=NoMirror
     resourceName=None
     resourceObject=&M_UI_Pilgrim_RotatorIntensity
     tiling=NoTile
     imageSize=<16 16>
     margin=<ltrb 0.5 0.5 0.5 0.5>
     outlineSettings {
      bUseBrushTransparency=false
      roundingType=HalfHeightRadius
      width=0
      color {colorUseRule=UseColor_Specified, specifiedColor=<rgba 0 0 0 0>}
      cornerRadii=<0 0 0 0>
     }
     tintColor {colorUseRule=UseColor_Specified, specifiedColor=<rgba 1 1 1 1>}
     uVRegion {bIsValid=false, max=<0 0>, min=<0 0>}
    }
    colorAndOpacity=<rgba 0.5 0.8333 1 1>
   }
   slot {class=OverlaySlot, properties {horizontalAlignment=HAlign_Fill, verticalAlignment=VAlign_Fill}}
  }
  Slot5 {
   widgetClassPath=&Overlay
   bIsVariable=true
   parent=HorizontalBox_116
   widget=Slot5
   slot {class=HorizontalBoxSlot, properties {verticalAlignment=VAlign_Bottom, padding=<ltrb 4 0 0 0>}}
  }
  Spacer_Slot5 {
   widgetClassPath=&Spacer
   bIsVariable=true
   parent=Slot5
   widget=Spacer_Slot5
   slot {class=OverlaySlot, properties {horizontalAlignment=HAlign_Fill, verticalAlignment=VAlign_Fill}}
  }
  Image_Slot5 {
   widgetClassPath=&Image
   bIsVariable=true
   parent=Slot5
   widget=Image_Slot5
   properties {
    renderOpacity=0.5
    visibility=SelfHitTestInvisible
    brush {
     bIsDynamicallyLoaded=false
     drawAs=Box
     imageType=FullColor
     mirroring=NoMirror
     resourceName=None
     resourceObject=&M_UI_Pilgrim_RotatorIntensity
     tiling=NoTile
     imageSize=<16 16>
     margin=<ltrb 0.5 0.5 0.5 0.5>
     outlineSettings {
      bUseBrushTransparency=false
      roundingType=HalfHeightRadius
      width=0
      color {colorUseRule=UseColor_Specified, specifiedColor=<rgba 0 0 0 0>}
      cornerRadii=<0 0 0 0>
     }
     tintColor {colorUseRule=UseColor_Specified, specifiedColor=<rgba 1 1 1 1>}
     uVRegion {bIsValid=false, max=<0 0>, min=<0 0>}
    }
    colorAndOpacity=<rgba 0.5 0.8333 1 1>
   }
   slot {class=OverlaySlot, properties {horizontalAlignment=HAlign_Fill, verticalAlignment=VAlign_Fill}}
  }
  Slot6 {
   widgetClassPath=&Overlay
   bIsVariable=true
   parent=HorizontalBox_116
   widget=Slot6
   slot {class=HorizontalBoxSlot, properties {verticalAlignment=VAlign_Bottom, padding=<ltrb 4 0 0 0>}}
  }
  Spacer_Slot6 {
   widgetClassPath=&Spacer
   bIsVariable=true
   parent=Slot6
   widget=Spacer_Slot6
   slot {class=OverlaySlot, properties {horizontalAlignment=HAlign_Fill, verticalAlignment=VAlign_Fill}}
  }
  Image_Slot6 {
   widgetClassPath=&Image
   bIsVariable=true
   parent=Slot6
   widget=Image_Slot6
   properties {
    renderOpacity=0.5
    visibility=SelfHitTestInvisible
    brush {
     bIsDynamicallyLoaded=false
     drawAs=Box
     imageType=FullColor
     mirroring=NoMirror
     resourceName=None
     resourceObject=&M_UI_Pilgrim_RotatorIntensity
     tiling=NoTile
     imageSize=<16 16>
     margin=<ltrb 0.5 0.5 0.5 0.5>
     outlineSettings {
      bUseBrushTransparency=false
      roundingType=HalfHeightRadius
      width=0
      color {colorUseRule=UseColor_Specified, specifiedColor=<rgba 0 0 0 0>}
      cornerRadii=<0 0 0 0>
     }
     tintColor {colorUseRule=UseColor_Specified, specifiedColor=<rgba 1 1 1 1>}
     uVRegion {bIsValid=false, max=<0 0>, min=<0 0>}
    }
    colorAndOpacity=<rgba 0.5 0.8333 1 1>
   }
   slot {class=OverlaySlot, properties {horizontalAlignment=HAlign_Fill, verticalAlignment=VAlign_Fill}}
  }
 ]
}
