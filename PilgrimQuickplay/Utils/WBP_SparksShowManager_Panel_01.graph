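# Sparks show manager widget panel with rotators for camera config, song section, color palette and a collapsed background selector.
# Camera clicks run a CameraDirector console command built from the rotator index.
# Other clicks update variables, then CallPilgrimManager disables stage updates on BP_PilgrimStageShowManager and forces a MIDI event, with palette index 5 remapped to 7.
# RunEntirePrimaryPanel applies all selections at once, and SetNavRules wraps navigation between top and bottom rotators.
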WidgetBlueprint /PilgrimQuickplay/Utils/WBP_SparksShowManager_Panel_01
graphs [
 CallPilgrimManager {nodes [
   CallPilgrimManager() {is=title K2Node_FunctionEntry}
   > Set EnableStageUpdate!(
    execute
    EnableStageUpdate:bool=false
    self:BP Pilgrim Stage Show Manager Object Reference=Get PilgrimManager() -> (PilgrimManager:BP Pilgrim Stage Show Manager Object Reference)
   ) -> (then, Output_Get:bool)
   > ForceMIDIEvent!(
    execute
    self:BP Pilgrim Stage Show Manager Object Reference=Get PilgrimManager()
    Mood:Name=Default
    Event:ESparksMidiTextEvents Enum=Get SongSection() -> (SongSection:ESparksMidiTextEvents Enum=None)
    ColorPaletteIndex:int=Get ColorPalette() -> (ColorPalette:int=0)
    ScreenIndex:int=Get StageBackground() -> (StageBackground:int=0)
   ) -> (then)
  ]}
 EventGraph {nodes [
   "On Button Clicked (Rotator_CameraConfig)"() {is=title K2Node_ComponentBoundEvent}
   n1 = int - int(A=Get RotatorIndex!(self:WBP Sparks Show Manager Rotator Object Reference=Get Rotator_CameraConfig() -> (Rotator_CameraConfig:WBP Sparks Show Manager Rotator Object Reference)) -> (RotatorIndex:int=0))
   > ExecuteConsoleCommand!(
    execute
    Command=Append(A=Get CustomConfigCmd() -> (CustomConfigCmd:String), B="To String (Integer)"(InInt=@n1))
    SpecificPlayer=GetOwningPlayer()
   ) -> (then)
   "On Button Clicked (Rotator_StageBackground)"() {is=title K2Node_ComponentBoundEvent}
   > Set StageBackground(StageBackground:int=Get RotatorIndex(self=Get Rotator_StageBackground() -> (Rotator_StageBackground:WBP Sparks Show Manager Rotator Object Reference)) -> (RotatorIndex=0)) -> (Output_Get:int=0)
   > CallPilgrimManager() {is=title K2Node_CallFunction}
   "On Button Clicked (Rotator_SongSection)"() {is=title K2Node_ComponentBoundEvent}
   > K2Node_Composite_3 = DetermineSongSection!(execute) -> (then:Exec) {is=title K2Node_Composite}
   > CallPilgrimManager()
   "On Button Clicked (Rotator_Color)"() {is=title K2Node_ComponentBoundEvent}
   n2 = Get RotatorIndex(self=Get Rotator_Color() -> (Rotator_Color:WBP Sparks Show Manager Rotator Object Reference)) -> (RotatorIndex=0)
   > n3 = Branch(Condition="Equal (Integer)"(A=@n2, B=5))
   > Set ColorPalette(ColorPalette:int=7) -> (Output_Get:int=0)
   > CallPilgrimManager()
   Set ColorPalette(execute=@n3.else, ColorPalette=@n2) -> (Output_Get=0)
   > CallPilgrimManager()
   RunEntirePrimaryPanel() {is=title K2Node_CustomEvent}
   n4 = int - int(A=Get RotatorIndex(self=Get Rotator_CameraConfig()) -> (RotatorIndex=0))
   > ExecuteConsoleCommand(Command=Append(A=Get CustomConfigCmd(), B="To String (Integer)"(InInt=@n4)), SpecificPlayer=GetOwningPlayer())
   > Set StageBackground(StageBackground=Get RotatorIndex(self=Get Rotator_StageBackground()) -> (RotatorIndex=0)) -> (Output_Get=0)
   n5 = Get Rotator_SongSection() -> (Rotator_SongSection:WBP Sparks Show Manager Rotator Object Reference)
   > n6 = Branch(Condition="Equal (Integer)"(A=Get RotatorIndex(self=@n5) -> (RotatorIndex=0)))
   > n7 = Set SongSection(SongSection:ESparksMidiTextEvents Enum=Intro) -> (Output_Get:ESparksMidiTextEvents Enum=None)
   n8 = Get RotatorIndex(self=Get Rotator_Color()) -> (RotatorIndex=0)
   n9 = Branch(execute=[@n7.then, @n14.then, @n17.then, @n20.then, @n22.then], Condition="Equal (Integer)"(A=@n8, B=5))
   > n10 = Set ColorPalette(ColorPalette=7) -> (Output_Get=0)
   CallPilgrimManager(execute=[@n10.then, @n11.then])
   n11 = Set ColorPalette(execute=@n9.else, ColorPalette=@n8) -> (Output_Get=0)
   n12 = Get Rotator_SongSection()
   n13 = Branch(execute=@n6.else, Condition="Equal (Integer)"(A=Get RotatorIndex(self=@n12) -> (RotatorIndex=0), B=1))
   > n14 = Set SongSection(SongSection=Verse) -> (Output_Get=None)
   n15 = Get Rotator_SongSection()
   n16 = Branch(execute=@n13.else, Condition="Equal (Integer)"(A=Get RotatorIndex(self=@n15) -> (RotatorIndex=0), B=2))
   > n17 = Set SongSection(SongSection=PreChorus) -> (Output_Get=None)
   n18 = Get Rotator_SongSection()
   n19 = Branch(execute=@n16.else, Condition="Equal (Integer)"(A=Get RotatorIndex(self=@n18) -> (RotatorIndex=0), B=3))
   > n20 = Set SongSection(SongSection=Chorus) -> (Output_Get=None)
   n21 = Get Rotator_SongSection()
   Branch(execute=@n19.else, Condition="Equal (Integer)"(A=Get RotatorIndex(self=@n21) -> (RotatorIndex=0), B=4))
   > n22 = Set SongSection(SongSection=Bridge) -> (Output_Get=None)
   Event Construct()
   > n23 = GetActorOfClass!(
    execute
    ActorClass:Actor Class Reference=/PilgrimQuickplay/StageManager/StageShowManager/BP_PilgrimStageShowManager.BP_PilgrimStageShowManager_C
   ) -> (then, ReturnValue:BP Pilgrim Stage Show Manager Object Reference)
   > Set PilgrimManager(PilgrimManager:BP Pilgrim Stage Show Manager Object Reference=@n23) -> (Output_Get:BP Pilgrim Stage Show Manager Object Reference)
  ]}
 DetermineSongSection {
  path=EventGraph.K2Node_Composite_3.DetermineSongSection
  nodes [
   Inputs()
   n1 = Get Rotator_SongSection()
   > n2 = Branch(Condition="Equal (Integer)"(A=Get RotatorIndex(self=@n1) -> (RotatorIndex=0)))
   > n3 = Set SongSection(SongSection=Intro) -> (Output_Get=None)
   Outputs!(then:Exec=[@n6.then, @n9.then, @n14.then, @n3.then, @n12.then])
   n4 = Get Rotator_SongSection()
   n5 = Branch(execute=@n2.else, Condition="Equal (Integer)"(A=Get RotatorIndex(self=@n4) -> (RotatorIndex=0), B=1))
   > n6 = Set SongSection(SongSection=Verse) -> (Output_Get=None)
   n7 = Get Rotator_SongSection()
   n8 = Branch(execute=@n5.else, Condition="Equal (Integer)"(A=Get RotatorIndex(self=@n7) -> (RotatorIndex=0), B=2))
   > n9 = Set SongSection(SongSection=PreChorus) -> (Output_Get=None)
   n10 = Get Rotator_SongSection()
   n11 = Branch(execute=@n8.else, Condition="Equal (Integer)"(A=Get RotatorIndex(self=@n10) -> (RotatorIndex=0), B=3))
   > n12 = Set SongSection(SongSection=Chorus) -> (Output_Get=None)
   n13 = Get Rotator_SongSection()
   Branch(execute=@n11.else, Condition="Equal (Integer)"(A=Get RotatorIndex(self=@n13) -> (RotatorIndex=0), B=4))
   > n14 = Set SongSection(SongSection=Bridge) -> (Output_Get=None)
  ]
 }
 SetNavRules {nodes [
   SetNavRules() {is=title K2Node_FunctionEntry}
   > SetNavigationRuleExplicit!(
    execute
    self:Widget Object Reference=Get Rotator_CameraConfig()
    Direction:EUINavigation Enum=Up
    InWidget:Widget Object Reference=Get Rotator_Color()
   ) -> (then)
   > SetNavigationRuleExplicit(self=Get Rotator_Color(), Direction=Down, InWidget=Get Rotator_CameraConfig())
  ]}
]
variables [
 ColorPalette {type=integer}
 CustomConfigCmd {type=string, default="Sparks.CameraDirector.CameraConfigEnabled.CustomConfig "}
 PilgrimManager {type=BP_PilgrimStageShowManager_C, default=null}
 SongSection {type=ESparksMidiTextEvents, default=ESparksMidiTextEvents_MAX}
 StageBackground {type=integer}
 VFX {type=integer}
]
widgets {
 info {inheritedWidgetCount=0, parentClass=&UserWidget, rootWidgetClass=&VerticalBox, widgetCount=5}
 widgets [
  VerticalBox {widgetClassPath=&VerticalBox, widget=VerticalBox}
  Rotator_CameraConfig {
   widgetClassPath=&/WBP_SparksShowManager_Rotator
   bIsVariable=true
   parent=VerticalBox
   widget=Rotator_CameraConfig
   properties {
    bIsFocusable=true
    currentText=Main
    title=Camera Config
    toolTipText=Cannot be changed at runtime
    rotatorOptions [Main, Estab, Close Ups, Mediums, Spotlights, All Custom]
   }
   slot {class=VerticalBoxSlot, properties {horizontalAlignment=HAlign_Center, padding=<ltrb 0 0 0 20>}}
  }
  Rotator_StageBackground {
   widgetClassPath=&WBP_SparksShowManager_Rotator
   bIsVariable=true
   parent=VerticalBox
   widget=Rotator_StageBackground
   properties {bIsFocusable=true, currentText=Default, title=Background, visibility=Collapsed, rotatorOptions [Eye,Spike,Diamond,Heart,Virus]}
   slot {class=VerticalBoxSlot, properties {horizontalAlignment=HAlign_Center, padding=<ltrb 0 0 0 20>}}
  }
  Rotator_SongSection {
   widgetClassPath=&WBP_SparksShowManager_Rotator
   bIsVariable=true
   parent=VerticalBox
   widget=Rotator_SongSection
   properties {bIsFocusable=true, currentText=None, title=Song Section, rotatorOptions [Intro,Verse,Prechorus,Chorus,Bridge]}
   slot {class=VerticalBoxSlot, properties {horizontalAlignment=HAlign_Center, padding=<ltrb 0 0 0 20>}}
  }
  Rotator_Color {
   widgetClassPath=&WBP_SparksShowManager_Rotator
   bIsVariable=true
   parent=VerticalBox
   widget=Rotator_Color
   properties {bIsFocusable=true, currentText=Default, title=Color Palette, rotatorOptions [1,2,3,4,5,6]}
   slot {class=VerticalBoxSlot, properties {horizontalAlignment=HAlign_Center, padding=<ltrb 0 0 0 20>}}
  }
 ]
}
