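# Secondary Sparks show manager debug panel widget with four rotators for local autoplay, global autoplay, auras and on-screen camera logging.
# Each rotator's click handler maps its index to an off or on console command run for the owning player, with global autoplay using a Cheat prefix.
# RunEntireSecondaryPanel reapplies all four settings in sequence.
# SetNavRules wraps navigation between the first and last rotators.
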WidgetBlueprint /PilgrimQuickplay/Utils/WBP_SparksShowManager_Panel_02
graphs [
 EventGraph {nodes [
   "On Button Clicked (Rotator_CameraLogToScreen)"() {is=title K2Node_ComponentBoundEvent}
   n1 = Get Rotator_CameraLogToScreen() -> (Rotator_CameraLogToScreen:WBP Sparks Show Manager Rotator Object Reference)
   > n2 = Branch(Condition="Equal (Integer)"(A=Get RotatorIndex!(self:WBP Sparks Show Manager Rotator Object Reference=@n1) -> (RotatorIndex:int=0)))
   > ExecuteConsoleCommand!(execute, Command=Sparks.CameraLogtoScreen 0, SpecificPlayer=GetOwningPlayer()) -> (then)
   ExecuteConsoleCommand(execute=@n2.else, Command=Sparks.CameraLogtoScreen 1, SpecificPlayer=GetOwningPlayer())
   "On Button Clicked (Rotator_Aura)"() {is=title K2Node_ComponentBoundEvent}
   n3 = Get Rotator_Aura() -> (Rotator_Aura:WBP Sparks Show Manager Rotator Object Reference)
   > n4 = Branch(Condition="Equal (Integer)"(A=Get RotatorIndex(self=@n3) -> (RotatorIndex=0)))
   > ExecuteConsoleCommand(Command=Sparks.ForceAura 0, SpecificPlayer=GetOwningPlayer())
   ExecuteConsoleCommand(execute=@n4.else, Command=Sparks.ForceAura 1, SpecificPlayer=GetOwningPlayer())
   "On Button Clicked (Rotator_LocalAutoplay)"() {is=title K2Node_ComponentBoundEvent}
   n5 = Get Rotator_LocalAutoplay() -> (Rotator_LocalAutoplay:WBP Sparks Show Manager Rotator Object Reference)
   > n6 = Branch(Condition="Equal (Integer)"(A=Get RotatorIndex(self=@n5) -> (RotatorIndex=0)))
   > ExecuteConsoleCommand(Command=Pilgrim.Autoplay 0, SpecificPlayer=GetOwningPlayer())
   ExecuteConsoleCommand(execute=@n6.else, Command=Pilgrim.Autoplay 1, SpecificPlayer=GetOwningPlayer())
   "On Button Clicked (Rotator_GlobalAutoplay)"() {is=title K2Node_ComponentBoundEvent}
   n7 = Get Rotator_GlobalAutoplay() -> (Rotator_GlobalAutoplay:WBP Sparks Show Manager Rotator Object Reference)
   > n8 = Branch(Condition="Equal (Integer)"(A=Get RotatorIndex(self=@n7) -> (RotatorIndex=0)))
   > ExecuteConsoleCommand(Command=Cheat Pilgrim.Autoplay 0, SpecificPlayer=GetOwningPlayer())
   ExecuteConsoleCommand(execute=@n8.else, Command=Cheat Pilgrim.Autoplay 1, SpecificPlayer=GetOwningPlayer())
   RunEntireSecondaryPanel() {is=title K2Node_CustomEvent}
   n9 = Get Rotator_LocalAutoplay()
   > n10 = Branch(Condition="Equal (Integer)"(A=Get RotatorIndex(self=@n9) -> (RotatorIndex=0)))
   > n11 = ExecuteConsoleCommand(Command=Pilgrim.Autoplay 0, SpecificPlayer=GetOwningPlayer())
   n12 = Get Rotator_GlobalAutoplay()
   n13 = Branch(execute=[@n11.then, @n22.then], Condition="Equal (Integer)"(A=Get RotatorIndex(self=@n12) -> (RotatorIndex=0)))
   > n14 = ExecuteConsoleCommand(Command=Cheat Pilgrim.Autoplay 0, SpecificPlayer=GetOwningPlayer())
   n15 = Get Rotator_Aura()
   n16 = Branch(execute=[@n14.then, @n21.then], Condition="Equal (Integer)"(A=Get RotatorIndex(self=@n15) -> (RotatorIndex=0)))
   > n17 = ExecuteConsoleCommand(Command=Sparks.ForceAura 0, SpecificPlayer=GetOwningPlayer())
   n18 = Get Rotator_CameraLogToScreen()
   n19 = Branch(execute=[@n17.then, @n20.then], Condition="Equal (Integer)"(A=Get RotatorIndex(self=@n18) -> (RotatorIndex=0)))
   > ExecuteConsoleCommand(Command=Sparks.CameraLogtoScreen 0, SpecificPlayer=GetOwningPlayer())
   ExecuteConsoleCommand(execute=@n19.else, Command=Sparks.CameraLogtoScreen 1, SpecificPlayer=GetOwningPlayer())
   n20 = ExecuteConsoleCommand(execute=@n16.else, Command=Sparks.ForceAura 1, SpecificPlayer=GetOwningPlayer())
   n21 = ExecuteConsoleCommand(execute=@n13.else, Command=Cheat Pilgrim.Autoplay 1, SpecificPlayer=GetOwningPlayer())
   n22 = ExecuteConsoleCommand(execute=@n10.else, Command=Pilgrim.Autoplay 1, SpecificPlayer=GetOwningPlayer())
  ]}
 SetNavRules {nodes [
   SetNavRules() {is=title K2Node_FunctionEntry}
   > SetNavigationRuleExplicit!(
    execute
    self:Widget Object Reference=Get Rotator_LocalAutoplay()
    Direction:EUINavigation Enum=Up
    InWidget:Widget Object Reference=Get Rotator_CameraLogToScreen()
   ) -> (then)
   > SetNavigationRuleExplicit(self=Get Rotator_CameraLogToScreen(), Direction=Down, InWidget=Get Rotator_LocalAutoplay())
  ]}
]
variables []
widgets {
 info {inheritedWidgetCount=0, parentClass=&UserWidget, rootWidgetClass=&VerticalBox, widgetCount=5}
 widgets [
  VerticalBox {widgetClassPath=&VerticalBox, widget=VerticalBox}
  Rotator_LocalAutoplay {
   widgetClassPath=&/WBP_SparksShowManager_Rotator
   bIsVariable=true
   parent=VerticalBox
   widget=Rotator_LocalAutoplay
   properties {bIsFocusable=true, currentText=Off, title=Local Autoplay, rotatorOptions [Off,On]}
   slot {class=VerticalBoxSlot, properties {horizontalAlignment=HAlign_Center, padding=<ltrb 0 0 0 20>}}
  }
  Rotator_GlobalAutoplay {
   widgetClassPath=&WBP_SparksShowManager_Rotator
   bIsVariable=true
   parent=VerticalBox
   widget=Rotator_GlobalAutoplay
   properties {bIsFocusable=true, currentText=Off, title=Global Autoplay, rotatorOptions [Off,On]}
   slot {class=VerticalBoxSlot, properties {horizontalAlignment=HAlign_Center, padding=<ltrb 0 0 0 20>}}
  }
  Rotator_Aura {
   widgetClassPath=&WBP_SparksShowManager_Rotator
   bIsVariable=true
   parent=VerticalBox
   widget=Rotator_Aura
   properties {bIsFocusable=true, currentText=Default, title=Auras, rotatorOptions [Default, Always On]}
   slot {class=VerticalBoxSlot, properties {horizontalAlignment=HAlign_Center, padding=<ltrb 0 0 0 20>}}
  }
  Rotator_CameraLogToScreen {
   widgetClassPath=&WBP_SparksShowManager_Rotator
   bIsVariable=true
   parent=VerticalBox
   widget=Rotator_CameraLogToScreen
   properties {bIsFocusable=true, currentText=Hide, title=Camera Info, toolTipText=Camera log to screen debugging tool, rotatorOptions [Hide,Show]}
   slot {class=VerticalBoxSlot, properties {horizontalAlignment=HAlign_Center, padding=<ltrb 0 0 0 20>}}
  }
 ]
}
