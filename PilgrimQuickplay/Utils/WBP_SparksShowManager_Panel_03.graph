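# Sparks Show Manager panel widget that sets stage positions for vocals, lead, bass and drums through Pilgrim cheat console commands.
# Clicking one of the four rotators reads its index, maps it to front, back, left or right via DeterminePosition, and executes that role's command.
# RunEntireTertiaryPanel executes all four commands in sequence from the current selections.
# SetNavRules links the top and bottom rotators so up/down navigation wraps around.
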WidgetBlueprint /PilgrimQuickplay/Utils/WBP_SparksShowManager_Panel_03
graphs [
 DeterminePosition {nodes [
   n1 = DeterminePosition!() -> (then, Index:int) {is=title K2Node_FunctionEntry}
   > n2 = Set LocalIndex(LocalIndex:int=@n1) -> (Output_Get:int=0)
   > n3 = Branch(Condition="Equal (Integer)"(A=@n2))
   > Return Node!(execute, Position:String=Get Front() -> (Front:String))
   n4 = Branch(execute=@n3.else, Condition="Equal (Integer)"(A=Get LocalIndex() -> (LocalIndex:int=0), B=1))
   > Return Node(Position=Get Back() -> (Back:String))
   n5 = Branch(execute=@n4.else, Condition="Equal (Integer)"(A=Get LocalIndex() -> (LocalIndex=0), B=2))
   > Return Node(Position=Get Left() -> (Left:String))
   Branch(execute=@n5.else, Condition="Equal (Integer)"(A=Get LocalIndex() -> (LocalIndex=0), B=3))
   > Return Node(Position=Get Right() -> (Right:String))
  ]}
 EventGraph {nodes [
   "On Button Clicked (Rotator_Vocals)"() {is=title K2Node_ComponentBoundEvent}
   n1 = Get RotatorIndex!(self:WBP Sparks Show Manager Rotator Object Reference=Get Rotator_Vocals() -> (Rotator_Vocals:WBP Sparks Show Manager Rotator Object Reference)) -> (RotatorIndex:int=0)
   > ExecuteConsoleCommand!(
    execute
    Command=Append(
     A=Get VocalPositionCmd() -> (VocalPositionCmd:String)
     B=DeterminePosition!(Index:int=@n1) -> (Position:String) {is=title K2Node_CallFunction}
    )
    SpecificPlayer
   ) -> (then)
   "On Button Clicked (Rotator_Lead)"() {is=title K2Node_ComponentBoundEvent}
   n2 = Get RotatorIndex(self=Get Rotator_Lead() -> (Rotator_Lead:WBP Sparks Show Manager Rotator Object Reference)) -> (RotatorIndex=0)
   > ExecuteConsoleCommand(Command=Append(A=Get LeadPositionCmd() -> (LeadPositionCmd:String), B=DeterminePosition(Index=@n2)))
   "On Button Clicked (Rotator_Bass)"() {is=title K2Node_ComponentBoundEvent}
   n3 = Get RotatorIndex(self=Get Rotator_Bass() -> (Rotator_Bass:WBP Sparks Show Manager Rotator Object Reference)) -> (RotatorIndex=0)
   > ExecuteConsoleCommand(Command=Append(A=Get BassPositionCmd() -> (BassPositionCmd:String), B=DeterminePosition(Index=@n3)))
   "On Button Clicked (Rotator_Drums)"() {is=title K2Node_ComponentBoundEvent}
   n4 = Get RotatorIndex(self=Get Rotator_Drums() -> (Rotator_Drums:WBP Sparks Show Manager Rotator Object Reference)) -> (RotatorIndex=0)
   > ExecuteConsoleCommand(Command=Append(A=Get DrumsPositionCmd() -> (DrumsPositionCmd:String), B=DeterminePosition(Index=@n4)))
   RunEntireTertiaryPanel() {is=title K2Node_CustomEvent}
   n5 = Get RotatorIndex(self=Get Rotator_Vocals()) -> (RotatorIndex=0)
   > ExecuteConsoleCommand(Command=Append(A=Get VocalPositionCmd(), B=DeterminePosition(Index=@n5)))
   n6 = Get RotatorIndex(self=Get Rotator_Lead()) -> (RotatorIndex=0)
   > ExecuteConsoleCommand(Command=Append(A=Get LeadPositionCmd(), B=DeterminePosition(Index=@n6)))
   n7 = Get RotatorIndex(self=Get Rotator_Bass()) -> (RotatorIndex=0)
   > ExecuteConsoleCommand(Command=Append(A=Get BassPositionCmd(), B=DeterminePosition(Index=@n7)))
   n8 = Get RotatorIndex(self=Get Rotator_Drums()) -> (RotatorIndex=0)
   > ExecuteConsoleCommand(Command=Append(A=Get DrumsPositionCmd(), B=DeterminePosition(Index=@n8)))
  ]}
 SetNavRules {nodes [
   SetNavRules() {is=title K2Node_FunctionEntry}
   > SetNavigationRuleExplicit!(
    execute
    self:Widget Object Reference=Get Rotator_Vocals()
    Direction:EUINavigation Enum=Up
    InWidget:Widget Object Reference=Get Rotator_Drums()
   ) -> (then)
   > SetNavigationRuleExplicit(self=Get Rotator_Drums(), Direction=Down, InWidget=Get Rotator_Vocals())
  ]}
]
variables [
 Back {type=string, default=back}
 BassPositionCmd {type=string, default="Cheat Pilgrim.BassStagePosition "}
 DrumsPositionCmd {type=string, default="Cheat Pilgrim.DrumsStagePosition "}
 Front {type=string, default=front}
 LeadPositionCmd {type=string, default="Cheat Pilgrim.GuitarStagePosition "}
 Left {type=string, default=left}
 Right {type=string, default=right}
 VocalPositionCmd {type=string, default="Cheat Pilgrim.VocalsStagePosition "}
]
widgets {
 info {inheritedWidgetCount=0, parentClass=&UserWidget, rootWidgetClass=&VerticalBox, widgetCount=5}
 widgets [
  VerticalBox {widgetClassPath=&VerticalBox, widget=VerticalBox}
  Rotator_Vocals {
   widgetClassPath=&/WBP_SparksShowManager_Rotator
   bIsVariable=true
   parent=VerticalBox
   widget=Rotator_Vocals
   properties {bIsFocusable=true, currentText=Random, title=Vocal Position, rotatorOptions [Front,Back,Left,Right]}
   slot {class=VerticalBoxSlot, properties {horizontalAlignment=HAlign_Center, padding=<ltrb 0 0 0 20>}}
  }
  Rotator_Lead {
   widgetClassPath=&WBP_SparksShowManager_Rotator
   bIsVariable=true
   parent=VerticalBox
   widget=Rotator_Lead
   properties {bIsFocusable=true, currentText=Random, title=Lead Position, rotatorOptions [Front,Back,Left,Right]}
   slot {class=VerticalBoxSlot, properties {horizontalAlignment=HAlign_Center, padding=<ltrb 0 0 0 20>}}
  }
  Rotator_Bass {
   widgetClassPath=&WBP_SparksShowManager_Rotator
   bIsVariable=true
   parent=VerticalBox
   widget=Rotator_Bass
   properties {bIsFocusable=true, currentText=Random, title=Bass Position, rotatorOptions [Front,Back,Left,Right]}
   slot {class=VerticalBoxSlot, properties {horizontalAlignment=HAlign_Center, padding=<ltrb 0 0 0 20>}}
  }
  Rotator_Drums {
   widgetClassPath=&WBP_SparksShowManager_Rotator
   bIsVariable=true
   parent=VerticalBox
   widget=Rotator_Drums
   properties {bIsFocusable=true, currentText=Random, title=Drums Position, rotatorOptions [Front,Back,Left,Right]}
   slot {class=VerticalBoxSlot, properties {horizontalAlignment=HAlign_Center, padding=<ltrb 0 0 0 20>}}
  }
 ]
}
