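# Modal prompt widget for the Sparks show manager where the player types a file name to save or load a settings preset.
# PreConstruct picks title and description text from bSave, falling back to bLoad, then InitAppearance applies them to the text blocks.
# Construct plays a short fade and scale Intro over a blurred backdrop, and Cancel removes the widget from its parent.
# The Accept button and text field have no logic in this graph.
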WidgetBlueprint /PilgrimQuickplay/Utils/WBP_SparksShowManager_Prompt
animations [Intro {
  binding_count=1
  length_seconds=0.25
  object=&self:Intro
  sequence {bindings [SizeBox_0 {
     id=#1
     possessed_class=SizeBox
     tracks [
      {
       type=MovieSceneFloatTrack
       display_name=Render Opacity
       sections [{type=MovieSceneFloatSection, end_frame=5, start_frame=0, channels [None {type=MovieSceneScriptingFloatChannel, keys [[0,0],[2,0],[5,1]]}]}]
      }
      {
       type=MovieScene2DTransformTrack
       sections [{
         type=MovieScene2DTransformSection
         end_frame=4
         start_frame=0
         channels [
          Translation.X {type=MovieSceneScriptingFloatChannel, const [0,1]}
          Translation.Y {type=MovieSceneScriptingFloatChannel, const [0,1]}
          Angle {type=MovieSceneScriptingFloatChannel, const [0,1]}
          Scale.X {type=MovieSceneScriptingFloatChannel, keys [[0,1.15],[2,1.15],[4,1]]}
          Scale.Y {type=MovieSceneScriptingFloatChannel, keys [[0,1.15],[2,1.15],[4,1]]}
          Shear.X {type=MovieSceneScriptingFloatChannel, const [0,1]}
          Shear.Y {type=MovieSceneScriptingFloatChannel, const [0,1]}
         ]
        }]
      }
     ]
    }]}
 }]
graphs [EventGraph {nodes [
   "On Clicked (Button_Cancel)"!() -> (OutputDelegate, then, Button:Common Button Base Object Reference) {is=title K2Node_ComponentBoundEvent}
   > CloseMenu()
   InitAppearance() {is=title K2Node_CustomEvent}
   > "SetText (Text)"(self=Get Text_Title() -> (Text_Title:Common Text Object Reference), InText=Get Title() -> (Title:Text))
   > "SetText (Text)"(self=Get Text_Description() -> (Text_Description:Common Text Object Reference), InText=Get Description() -> (Description:Text))
   CloseMenu() {is=title K2Node_CustomEvent}
   > RemoveFromParent(self=Self-Reference())
   Event PreConstruct()
   > n1 = Branch(Condition=Get bSave() -> (bSave:bool))
   > Set Title(Title:Text="LOCTEXT(\"Save Settings As Preset\")") -> (Output_Get:Text)
   > Set Description(Description:Text="LOCTEXT(\"Type in a name for this save file\")") -> (Output_Get:Text)
   > InitAppearance() {is=title K2Node_CallFunction}
   Branch(execute=@n1.else, Condition=Get bLoad() -> (bLoad:bool))
   > Set Title(Title="LOCTEXT(\"Load Preset From Save\")")
   > Set Description(Description="LOCTEXT(\"Type in the name of the save file you want to load\")")
   > InitAppearance()
   Event Construct()
   > PlayAnimationForward(InAnimation=Get Intro())
  ]}]
variables [
 bLoad {type=boolean}
 bSave {type=boolean, default=true}
 Description {type=string, default=""}
 Title {type=string, default=Save Settings As Preset}
]
widgets {
 info {inheritedWidgetCount=0, parentClass=&UserWidget, rootWidgetClass=&CanvasPanel, widgetCount=23}
 widgets [
  CanvasPanel_76 {widgetClassPath=&CanvasPanel, widget=CanvasPanel_76}
  Overlay_189 {
   widgetClassPath=&Overlay
   parent=CanvasPanel_76
   widget=Overlay_189
   slot {class=CanvasPanelSlot, properties {layoutData {alignment=<0 0>, anchors {maximum=<1 1>, minimum=<0 0>}, offsets=<ltrb 0 0 0 0>}}}
  }
  Blur {
   widgetClassPath=&BackgroundBlur
   parent=Overlay_189
   widget=Blur
   properties {blurStrength=15}
   slot {class=OverlaySlot, properties {horizontalAlignment=HAlign_Fill, verticalAlignment=VAlign_Fill}}
  }
  Background {
   widgetClassPath=&Border
   parent=Overlay_189
   widget=Background
   properties {
    background {
     bIsDynamicallyLoaded=false
     drawAs=Image
     imageType=NoImage
     mirroring=NoMirror
     resourceName=None
     resourceObject=null
     tiling=NoTile
     imageSize=<32 32>
     margin=<ltrb 0 0 0 0>
     outlineSettings {
      bUseBrushTransparency=false
      roundingType=HalfHeightRadius
      width=0
      color {colorUseRule=UseColor_Specified, specifiedColor=<rgba 0 0 0 0>}
      cornerRadii=<0 0 0 0>
     }
     tintColor {colorUseRule=UseColor_Specified, specifiedColor=<rgba 0 0.0006 0.0075 1>}
     uVRegion {bIsValid=false, max=<0 0>, min=<0 0>}
    }
    contentColorAndOpacity=<rgba 0 0.0006 0.0075 1>
    padding=<ltrb 0 0 0 0>
   }
   slot {class=OverlaySlot, properties {horizontalAlignment=HAlign_Fill, verticalAlignment=VAlign_Fill}}
  }
  Gradient {
   widgetClassPath=&Border
   bIsVariable=true
   parent=Overlay_189
   widget=Gradient
   properties {
    visibility=SelfHitTestInvisible
    background {
     bIsDynamicallyLoaded=false
     drawAs=Image
     imageType=NoImage
     mirroring=NoMirror
     resourceName=None
     resourceObject=/Game/Athena/UI/Frontend/Art/M_UI_RadialGradientMaster_Inst
     tiling=NoTile
     imageSize=<32 32>
     margin=<ltrb 0 0 0 0>
     outlineSettings {
      bUseBrushTransparency=false
      roundingType=HalfHeightRadius
      width=0
      color {colorUseRule=UseColor_Specified, specifiedColor=<rgba 0 0 0 0>}
      cornerRadii=<0 0 0 0>
     }
     tintColor {colorUseRule=UseColor_Specified, specifiedColor=<rgba 1 1 1 1>}
     uVRegion {bIsValid=false, max=<0 0>, min=<0 0>}
    }
    brushColor=<rgba 0.0869 0.0869 0.094 0.5>
    padding=<ltrb 0 0 0 0>
   }
   slot {class=OverlaySlot, properties {horizontalAlignment=HAlign_Fill, verticalAlignment=VAlign_Fill}}
  }
  SizeBox_0 {
   widgetClassPath=&SizeBox
   parent=Overlay_189
   widget=SizeBox_0
   properties {
    bOverride_MinDesiredHeight=true
    bOverride_WidthOverride=true
    heightOverride=800
    minDesiredHeight=360
    minDesiredWidth=1000
    widthOverride=1000
   }
   slot {class=OverlaySlot, properties {horizontalAlignment=HAlign_Center, verticalAlignment=VAlign_Center}}
  }
  CommonBorder_1 {
   widgetClassPath=&CommonBorder
   parent=SizeBox_0
   widget=CommonBorder_1
   properties {
    background {
     bIsDynamicallyLoaded=false
     drawAs=RoundedBox
     imageType=NoImage
     mirroring=NoMirror
     resourceName=None
     resourceObject=null
     tiling=NoTile
     imageSize=<64 64>
     margin=<ltrb 0 0 0 0>
     outlineSettings {
      bUseBrushTransparency=false
      roundingType=FixedRadius
      width=0
      color {colorUseRule=UseColor_Specified, specifiedColor=<rgba 0 0 0 0>}
      cornerRadii=<25 25 25 25>
     }
     tintColor {colorUseRule=UseColor_Specified, specifiedColor=<rgba 0.0409 0.0168 0.1683 0.92>}
     uVRegion {bIsValid=false, max=<0 0>, min=<0 0>}
    }
    padding=<ltrb 0 0 0 0>
   }
   slot {class=SizeBoxSlot}
  }
  Overlay_0 {widgetClassPath=&Overlay, parent=CommonBorder_1, widget=Overlay_0, slot {class=BorderSlot, properties {padding=<ltrb 0 0 0 0>}}}
  VerticalBox_1 {
   widgetClassPath=&VerticalBox
   parent=Overlay_0
   widget=VerticalBox_1
   slot {class=OverlaySlot, properties {horizontalAlignment=HAlign_Fill, verticalAlignment=VAlign_Fill}}
  }
  CommonBorder_Modal {
   widgetClassPath=&CommonBorder
   parent=VerticalBox_1
   widget=CommonBorder_Modal
   properties {
    background {
     bIsDynamicallyLoaded=false
     drawAs=RoundedBox
     imageType=NoImage
     mirroring=NoMirror
     resourceName=None
     resourceObject=null
     tiling=NoTile
     imageSize=<64 64>
     margin=<ltrb 0 0 0 0>
     outlineSettings {
      bUseBrushTransparency=false
      roundingType=FixedRadius
      width=0
      color {colorUseRule=UseColor_Specified, specifiedColor=<rgba 0 0 0 0>}
      cornerRadii=<25 25 0 0>
     }
     tintColor {colorUseRule=UseColor_Specified, specifiedColor=<rgba 0.0086 0.0097 0.0865 0.82>}
     uVRegion {bIsValid=false, max=<0 0>, min=<0 0>}
    }
    padding=<ltrb 40 0 40 0>
   }
   slot {class=VerticalBoxSlot}
  }
  HorizontalBox_1 {
   widgetClassPath=&HorizontalBox
   parent=CommonBorder_Modal
   widget=HorizontalBox_1
   slot {class=BorderSlot, properties {padding=<ltrb 40 0 40 0>}}
  }
  Spacer_0 {widgetClassPath=&Spacer, parent=HorizontalBox_1, widget=Spacer_0, properties {size=<0 85>}, slot {class=HorizontalBoxSlot}}
  Text_Title {
   widgetClassPath=&CommonTextBlock
   bIsVariable=true
   parent=HorizontalBox_1
   widget=Text_Title
   properties {
    text=Title
    textTransformPolicy=ToUpper
    font {
     bForceMonospaced=false
     bMaterialIsStencil=false
     fontMaterial=null
     fontObject=&HeadingNow
     letterSpacing=0
     monospacedWidth=1
     size=33
     skewAmount=0
     typefaceFontName=76 Bold Italic
     outlineSettings {
      bApplyOutlineToDropShadows=false
      bMiteredCorners=false
      bSeparateFillAlpha=false
      outlineMaterial=null
      outlineSize=0
      outlineColor=<rgba 0 0 0 1>
     }
    }
    shadowOffset=<0 0>
    strikeBrush {
     bIsDynamicallyLoaded=false
     drawAs=NoDrawType
     imageType=NoImage
     mirroring=NoMirror
     resourceName=None
     resourceObject=null
     tiling=NoTile
     imageSize=<32 32>
     margin=<ltrb 0 0 0 0>
     outlineSettings {
      bUseBrushTransparency=false
      roundingType=HalfHeightRadius
      width=0
      color {colorUseRule=UseColor_Specified, specifiedColor=<rgba 0 0 0 0>}
      cornerRadii=<0 0 0 0>
     }
     tintColor {colorUseRule=UseColor_Specified, specifiedColor=<rgba 1 1 1 1>}
     uVRegion {bIsValid=false, max=<0 0>, min=<0 0>}
    }
   }
   slot {class=HorizontalBoxSlot, properties {verticalAlignment=VAlign_Center, padding=<ltrb 8 0 20 0>}}
  }
  Grid_Content {
   widgetClassPath=&GridPanel
   parent=VerticalBox_1
   widget=Grid_Content
   properties {columnFill [1], rowFill [1]}
   slot {class=VerticalBoxSlot, properties {verticalAlignment=VAlign_Top, padding=<ltrb 48 32 48 32>, size {sizeRule=Fill, value=1}}}
  }
  Spacer_Text {
   widgetClassPath=&Spacer
   parent=Grid_Content
   widget=Spacer_Text
   properties {size=<214 60>}
   slot {class=GridSlot, properties {column=1, horizontalAlignment=HAlign_Right, verticalAlignment=VAlign_Top}}
  }
  Text_Description {
   widgetClassPath=&CommonTextBlock
   bIsVariable=true
   parent=Grid_Content
   widget=Text_Description
   properties {
    autoWrapText=true
    lineHeightPercentage=1.29
    text=Enter a game session ID:
    textTransformPolicy=ToUpper
    colorAndOpacity {colorUseRule=UseColor_Specified, specifiedColor=<rgba 0.5 0.8333 1 1>}
    font {
     bForceMonospaced=false
     bMaterialIsStencil=false
     fontMaterial=null
     fontObject=&HeadingNow
     letterSpacing=0
     monospacedWidth=1
     size=15
     skewAmount=0
     typefaceFontName=76 Bold Italic
     outlineSettings {
      bApplyOutlineToDropShadows=false
      bMiteredCorners=false
      bSeparateFillAlpha=false
      outlineMaterial=null
      outlineSize=0
      outlineColor=<rgba 0 0 0 1>
     }
    }
    shadowOffset=<0 0>
    strikeBrush {
     bIsDynamicallyLoaded=false
     drawAs=NoDrawType
     imageType=NoImage
     mirroring=NoMirror
     resourceName=None
     resourceObject=null
     tiling=NoTile
     imageSize=<32 32>
     margin=<ltrb 0 0 0 0>
     outlineSettings {
      bUseBrushTransparency=false
      roundingType=HalfHeightRadius
      width=0
      color {colorUseRule=UseColor_Specified, specifiedColor=<rgba 0 0 0 0>}
      cornerRadii=<0 0 0 0>
     }
     tintColor {colorUseRule=UseColor_Specified, specifiedColor=<rgba 1 1 1 1>}
     uVRegion {bIsValid=false, max=<0 0>, min=<0 0>}
    }
   }
   slot {class=GridSlot, properties {verticalAlignment=VAlign_Top}}
  }
  EditableText {
   widgetClassPath=&CommonBorder
   parent=Grid_Content
   widget=EditableText
   properties {
    verticalAlignment=VAlign_Center
    background {
     bIsDynamicallyLoaded=false
     drawAs=Image
     imageType=NoImage
     mirroring=NoMirror
     resourceName=None
     resourceObject=null
     tiling=NoTile
     imageSize=<32 32>
     margin=<ltrb 0 0 0 0>
     outlineSettings {
      bUseBrushTransparency=false
      roundingType=HalfHeightRadius
      width=0
      color {colorUseRule=UseColor_Specified, specifiedColor=<rgba 0 0 0 0>}
      cornerRadii=<0 0 0 0>
     }
     tintColor {colorUseRule=UseColor_Specified, specifiedColor=<rgba 1 1 1 1>}
     uVRegion {bIsValid=false, max=<0 0>, min=<0 0>}
    }
    padding=<ltrb 16 0 16 0>
   }
   slot {class=GridSlot, properties {layer=20, row=1, verticalAlignment=VAlign_Top, padding=<ltrb 0 0 40 0>}}
  }
  HorizontalBox_0 {
   widgetClassPath=&HorizontalBox
   parent=EditableText
   widget=HorizontalBox_0
   slot {class=BorderSlot, properties {verticalAlignment=VAlign_Center, padding=<ltrb 16 0 16 0>}}
  }
  EditableText_FileName {
   widgetClassPath=/Script/UMG.EditableText
   bIsVariable=true
   parent=HorizontalBox_0
   widget=EditableText_FileName
   properties {
    hintText=Enter text here:
    navigation=&self:WidgetTree.EditableText_FileName.WidgetNavigation_0
    widgetStyle {
     backgroundImageComposing {
      bIsDynamicallyLoaded=false
      drawAs=NoDrawType
      imageType=FullColor
      mirroring=NoMirror
      resourceName=../../../Engine/Content/Slate/Old/HyperlinkDotted.png
      resourceObject=null
      tiling=Both
      imageSize=<0 0>
      margin=<ltrb 0 0 0 0.1875>
      outlineSettings {
       bUseBrushTransparency=false
       roundingType=HalfHeightRadius
       width=0
       color {colorUseRule=UseColor_Specified, specifiedColor=<rgba 0 0 0 0>}
       cornerRadii=<0 0 0 0>
      }
      tintColor {colorUseRule=UseColor_Specified, specifiedColor=<rgba 1 1 1 1>}
      uVRegion {bIsValid=false, max=<0 0>, min=<0 0>}
     }
     backgroundImageSelected {
      bIsDynamicallyLoaded=false
      drawAs=NoDrawType
      imageType=FullColor
      mirroring=NoMirror
      resourceName=../../../Engine/Content/Slate/Common/EditableTextSelectionBackground.png
      resourceObject=null
      tiling=NoTile
      imageSize=<0 0>
      margin=<ltrb 0.25 0.25 0.25 0.25>
      outlineSettings {
       bUseBrushTransparency=false
       roundingType=HalfHeightRadius
       width=0
       color {colorUseRule=UseColor_Specified, specifiedColor=<rgba 0 0 0 0>}
       cornerRadii=<0 0 0 0>
      }
      tintColor {colorUseRule=UseColor_Specified, specifiedColor=<rgba 1 1 1 1>}
      uVRegion {bIsValid=false, max=<0 0>, min=<0 0>}
     }
     caretImage {
      bIsDynamicallyLoaded=false
      drawAs=Image
      imageType=FullColor
      mirroring=NoMirror
      resourceName=../../../Engine/Content/Slate/Old/White.png
      resourceObject=null
      tiling=NoTile
      imageSize=<16 16>
      margin=<ltrb 0 0 0 0>
      outlineSettings {
       bUseBrushTransparency=false
       roundingType=HalfHeightRadius
       width=0
       color {colorUseRule=UseColor_Specified, specifiedColor=<rgba 0 0 0 0>}
       cornerRadii=<0 0 0 0>
      }
      tintColor {colorUseRule=UseColor_Specified, specifiedColor=<rgba 1 1 1 1>}
      uVRegion {bIsValid=false, max=<0 0>, min=<0 0>}
     }
     colorAndOpacity {colorUseRule=UseColor_Specified, specifiedColor=<rgba 0 0 0 1>}
     font {
      bForceMonospaced=false
      bMaterialIsStencil=false
      fontMaterial=null
      fontObject=&HeadingNow
      letterSpacing=0
      monospacedWidth=1
      size=13.5
      skewAmount=0
      typefaceFontName=76 Bold
      outlineSettings {
       bApplyOutlineToDropShadows=false
       bMiteredCorners=false
       bSeparateFillAlpha=false
       outlineMaterial=null
       outlineSize=0
       outlineColor=<rgba 0 0 0 1>
      }
     }
     hintColorAndOpacity {colorUseRule=UseColor_Foreground, specifiedColor=<rgba 1 0 1 1>}
     selectedBackgroundColor {colorUseRule=UseColor_Foreground, specifiedColor=<rgba 1 0 1 1>}
    }
   }
   slot {class=HorizontalBoxSlot, properties {verticalAlignment=VAlign_Center, size {sizeRule=Fill, value=1}}}
  }
  Spacer_EditText {
   widgetClassPath=&Spacer
   parent=HorizontalBox_0
   widget=Spacer_EditText
   properties {size=<0 56>}
   slot {class=HorizontalBoxSlot, properties {horizontalAlignment=HAlign_Left, verticalAlignment=VAlign_Top}}
  }
  VerticalBox_0 {
   widgetClassPath=&VerticalBox
   parent=Grid_Content
   widget=VerticalBox_0
   slot {class=GridSlot, properties {column=1, layer=30, row=1, rowSpan=2}}
  }
  Button_SaveLoad {
   widgetClassPath=/FortUILibrary/UIKit/Buttons/WBP_UIKit_Button_Loud
   bIsVariable=true
   parent=VerticalBox_0
   widget=Button_SaveLoad
   properties {text=Accept}
   slot {class=VerticalBoxSlot, properties {padding=<ltrb 0 0 0 24>}}
  }
  Button_Cancel {
   widgetClassPath=&WBP_UIKit_Button_Regular
   bIsVariable=true
   parent=VerticalBox_0
   widget=Button_Cancel
   properties {bHideInputActionWithKeyboard=true, text=Cancel}
   slot {class=VerticalBoxSlot}
  }
 ]
}
