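# Three-icon tab bar widget for the Sparks Show Manager UI in PilgrimQuickplay, built from selectable UIKit buttons.
# Clicking a tab selects it, deselects the previous one via GetButtonIndex and DeselectTab, then stores Index; PreConstruct starts on the first tab.
# Hover events drive button block visual states, restoring idle only on unselected tabs.
# Construct sets explicit left/right navigation wrapping between the end tabs; no content switching happens here.
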WidgetBlueprint /PilgrimQuickplay/Utils/WBP_SparksShowManager_Tabs
graphs [
 DeselectTab {nodes [
   n1 = DeselectTab!() -> (then, Tab:Common Button Base Object Reference) {is=title K2Node_FunctionEntry}
   > ClearSelection(self:Common Button Base Object Reference=@n1)
   > SetIsSelected!(execute, self:Common Button Base Object Reference=@n1, InSelected:bool=false, bGiveClickFeedback:bool=true) -> (then)
  ]}
 EventGraph {nodes [
   "On Clicked (Button_Tab2)"!() -> (OutputDelegate, then, Button:Common Button Base Object Reference) {is=title K2Node_ComponentBoundEvent}
   > SetIsSelected(self=Get Button_Tab2() -> (Button_Tab2:WBP UIKit Button Selectable Object Reference), InSelected=true, bGiveClickFeedback=true)
   > n1 = Branch(Condition="Equal (Integer)"(A=Get Index() -> (Index:int=0), B=1))
   n2 = GetButtonIndex!(execute=@n1.else, self, CurrentlySelectedSlot:int=Get Index() -> (Index=0)) -> (then, CurrentlySelectedTab:WBP UIKit Button Selectable Object Reference)
   > DeselectTab!(execute, self, Tab:Common Button Base Object Reference=@n2) -> (then) {is=title K2Node_CallFunction}
   > Set Index(Index:int=1) -> (Output_Get:int=0)
   "On Clicked (Button_Tab1)"!() -> (OutputDelegate, then, Button:Common Button Base Object Reference) {is=title K2Node_ComponentBoundEvent}
   > SetIsSelected(self=Get Button_Tab1() -> (Button_Tab1:WBP UIKit Button Selectable Object Reference), InSelected=true, bGiveClickFeedback=true)
   > n3 = Branch(Condition="Equal (Integer)"(A=Get Index() -> (Index=0)))
   n4 = GetButtonIndex(execute=@n3.else, CurrentlySelectedSlot=Get Index() -> (Index=0))
   > DeselectTab(Tab=@n4)
   > Set Index(Index=0) -> (Output_Get=0)
   "On Clicked (Button_Tab3)"!() -> (OutputDelegate, then, Button:Common Button Base Object Reference) {is=title K2Node_ComponentBoundEvent}
   > SetIsSelected(self=Get Button_Tab3() -> (Button_Tab3:WBP UIKit Button Selectable Object Reference), InSelected=true, bGiveClickFeedback=true)
   > n5 = Branch(Condition="Equal (Integer)"(A=Get Index() -> (Index=0), B=2))
   n6 = GetButtonIndex(execute=@n5.else, CurrentlySelectedSlot=Get Index() -> (Index=0))
   > DeselectTab(Tab=@n6)
   > Set Index(Index=2) -> (Output_Get=0)
   "On Unhovered (Button_Tab1)"!() -> (OutputDelegate, then, Button:Common Button Base Object Reference) {is=title K2Node_ComponentBoundEvent}
   > Branch(Condition="Not Equal (Integer)"(A=Get Index() -> (Index=0)))
   > RequestTransitionOnBlocks!(execute, self:UIKit Modular Button Object Reference=Get Button_Tab1(), State:EUIKitBlockVisualState Enum=DeselectedIdle) -> (then)
   "On Hovered (Button_Tab1)"!() -> (OutputDelegate, then, Button:Common Button Base Object Reference) {is=title K2Node_ComponentBoundEvent}
   > RequestTransitionOnBlocks(self=Get Button_Tab1(), State=Hovered)
   "On Hovered (Button_Tab2)"!() -> (OutputDelegate, then, Button:Common Button Base Object Reference) {is=title K2Node_ComponentBoundEvent}
   > RequestTransitionOnBlocks(self=Get Button_Tab2(), State=Hovered)
   "On Unhovered (Button_Tab2)"!() -> (OutputDelegate, then, Button:Common Button Base Object Reference) {is=title K2Node_ComponentBoundEvent}
   > Branch(Condition="Not Equal (Integer)"(A=Get Index() -> (Index=0), B=1))
   > RequestTransitionOnBlocks(self=Get Button_Tab2(), State=DeselectedIdle)
   "On Hovered (Button_Tab3)"!() -> (OutputDelegate, then, Button:Common Button Base Object Reference) {is=title K2Node_ComponentBoundEvent}
   > RequestTransitionOnBlocks(self=Get Button_Tab3(), State=Hovered)
   "On Unhovered (Button_Tab3)"!() -> (OutputDelegate, then, Button:Common Button Base Object Reference) {is=title K2Node_ComponentBoundEvent}
   > Branch(Condition="Not Equal (Integer)"(A=Get Index() -> (Index=0), B=2))
   > RequestTransitionOnBlocks(self=Get Button_Tab3(), State=DeselectedIdle)
   Event PreConstruct()
   > SetIsSelected(self=Get Button_Tab1(), InSelected=true, bGiveClickFeedback=true)
   > Set Index(Index=0) -> (Output_Get=0)
   Event Construct()
   > SetNavRules()
  ]}
 GetButtonIndex {nodes [
   n1 = GetButtonIndex!() -> (then, CurrentlySelectedSlot:int) {is=title K2Node_FunctionEntry}
   n2 = Get Button_Tab1()
   n3 = Get Button_Tab2()
   n4 = Get Button_Tab3()
   > Return Node!(
    execute
    CurrentlySelectedTab:WBP UIKit Button Selectable Object Reference="Get (a copy)"(
     Array:Array of WBP UIKit Button Selectable Object References=Make Array!(
      "[0]":WBP UIKit Button Selectable Object Reference=@n2
      "[1]":WBP UIKit Button Selectable Object Reference=@n3
      "[2]":WBP UIKit Button Selectable Object Reference=@n4
     ) -> (Array:Array of WBP UIKit Button Selectable Object References)
     Dimension 1=@n1
    ) -> (Output:WBP UIKit Button Selectable Object Reference)
   )
  ]}
 SetNavRules {nodes [
   SetNavRules() {is=title K2Node_FunctionEntry}
   > SetNavigationRuleExplicit!(
    execute
    self:Widget Object Reference=Get Button_Tab1()
    Direction:EUINavigation Enum=Left
    InWidget:Widget Object Reference=Get Button_Tab3()
   ) -> (then)
   > SetNavigationRuleExplicit(self=Get Button_Tab3(), Direction=Right, InWidget=Get Button_Tab1())
  ]}
]
variables [Index {type=integer}]
widgets {
 info {inheritedWidgetCount=0, parentClass=&UserWidget, rootWidgetClass=&SizeBox, widgetCount=8}
 widgets [
  SizeBox_33 {widgetClassPath=&SizeBox, widget=SizeBox_33}
  Border_101 {
   widgetClassPath=&Border
   parent=SizeBox_33
   widget=Border_101
   properties {
    horizontalAlignment=HAlign_Center
    verticalAlignment=VAlign_Center
    background {
     bIsDynamicallyLoaded=false
     drawAs=Image
     imageType=FullColor
     mirroring=NoMirror
     resourceName=None
     resourceObject=/PilgrimCore/Widgets/Materials/MI_Pilgrim_TabBacking_Panel
     tiling=NoTile
     imageSize=<32 32>
     margin=<ltrb 0 0 0 0>
     outlineSettings {
      bUseBrushTransparency=false
      roundingType=HalfHeightRadius
      width=0
      color {colorUseRule=UseColor_Specified, specifiedColor=<rgba 0 0 0 0>}
      cornerRadii=<0 0 0 0>
     }
     tintColor {colorUseRule=UseColor_Specified, specifiedColor=<rgba 1 1 1 1>}
     uVRegion {bIsValid=false, max=<0 0>, min=<0 0>}
    }
    padding=<ltrb 8 8 8 8>
   }
   slot {class=SizeBoxSlot}
  }
  ScaleBox_4 {
   widgetClassPath=&ScaleBox
   parent=Border_101
   widget=ScaleBox_4
   slot {class=BorderSlot, properties {horizontalAlignment=HAlign_Center, verticalAlignment=VAlign_Center, padding=<ltrb 8 8 8 8>}}
  }
  VerticalBox_0 {widgetClassPath=&VerticalBox, parent=ScaleBox_4, widget=VerticalBox_0, slot {class=ScaleBoxSlot}}
  HorizontalBox_116 {widgetClassPath=&HorizontalBox, parent=VerticalBox_0, widget=HorizontalBox_116, slot {class=VerticalBoxSlot}}
  Button_Tab1 {
   widgetClassPath=/FortUILibrary/UIKit/Buttons/WBP_UIKit_Button_Selectable
   bIsVariable=true
   parent=HorizontalBox_116
   widget=Button_Tab1
   properties {
    bTriggerClickedAfterSelection=true
    backgroundClass=/FortUILibrary/Systems/ModularButton/Blocks/WBP_UIKit_Block_BackgroundGeneric.WBP_UIKit_Block_BackgroundGeneric_C
    backgroundMaterial=/PilgrimCore/Widgets/Materials/MI_UI_Piligrim_IconTabBacking
    iconHorizontalAlignment=HAlign_Left
    iconMaterial=&/MI_UI_Icon_ShowManagerTab_01
    minHeight=0
    outlineMaterial=/FortUILibrary/UIKit/Buttons/MaterialInstances/CTAButtons/MI_UIKit_EmptyProgress
    text=0
    useTextBlock=false
    buttonPadding=<ltrb 5 5 5 5>
    iconPadding=<ltrb 8 0 8 0>
    iconSize=<72 72>
   }
   slot {class=HorizontalBoxSlot, properties {verticalAlignment=VAlign_Center, padding=<ltrb 0 0 35 0>}}
  }
  Button_Tab2 {
   widgetClassPath=&WBP_UIKit_Button_Selectable
   bIsVariable=true
   parent=HorizontalBox_116
   widget=Button_Tab2
   properties {
    bTriggerClickedAfterSelection=true
    backgroundClass=&WBP_UIKit_Block_BackgroundGeneric_C
    backgroundMaterial=&MI_UI_Piligrim_IconTabBacking
    iconHorizontalAlignment=HAlign_Left
    iconMaterial=&/MI_UI_Icon_ShowManagerTab_02
    minHeight=0
    outlineMaterial=&MI_UIKit_EmptyProgress
    text=0
    useTextBlock=false
    buttonPadding=<ltrb 5 5 5 5>
    iconPadding=<ltrb 8 0 8 0>
    iconSize=<72 72>
   }
   slot {class=HorizontalBoxSlot, properties {verticalAlignment=VAlign_Center, padding=<ltrb 0 0 35 0>}}
  }
  Button_Tab3 {
   widgetClassPath=&WBP_UIKit_Button_Selectable
   bIsVariable=true
   parent=HorizontalBox_116
   widget=Button_Tab3
   properties {
    bTriggerClickedAfterSelection=true
    backgroundClass=&WBP_UIKit_Block_BackgroundGeneric_C
    backgroundMaterial=&MI_UI_Piligrim_IconTabBacking
    iconHorizontalAlignment=HAlign_Left
    iconMaterial=&/MI_UI_Icon_ShowManagerTab_03
    minHeight=0
    outlineMaterial=&MI_UIKit_EmptyProgress
    text=0
    useTextBlock=false
    buttonPadding=<ltrb 5 5 5 5>
    iconPadding=<ltrb 8 0 8 0>
    iconSize=<72 72>
   }
   slot {class=HorizontalBoxSlot, properties {verticalAlignment=VAlign_Center}}
  }
 ]
}
