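# Sparks Show Manager menu widget: a tabbed overlay of rotator settings for camera, stage, autoplay, debug options and instrument positions.
# Reset restores the first two tabs through the stage show manager or console commands; Exit toggles the capture tool.
# Save and Load use a filename prompt and SaveGame slot storing rotator indices; loading reapplies them by array position, then runs all three panels.
# Gamepad input gets explicit focus and hides Save/Load.
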WidgetBlueprint /PilgrimQuickplay/Utils/WBP_SparksShowManager_Menu
graphs [
 ApplyRotatorPositionsFromSave {nodes [
   n1 = ApplyRotatorPositionsFromSave!() -> (then, ArrayElement:int, ArrayIndex:int) {is=title K2Node_FunctionEntry}
   > Set ArrayElementLoc(ArrayElementLoc:int=@n1) -> (Output_Get:int=0)
   > Set ArrayIndexLoc(ArrayIndexLoc:int=@n1.ArrayIndex) -> (Output_Get:int=0)
   > n2 = Branch(Condition="Equal (Integer)"(A=Get ArrayIndexLoc() -> (ArrayIndexLoc:int=0)))
   > SetRotatorToSpecificIndex!(
    execute
    self:WBP Sparks Show Manager Rotator Object Reference=Get Rotator_CameraConfig!(self:WBP Sparks Show Manager Panel 01 Object Reference=Get Panel_Primary() -> (Panel_Primary:WBP Sparks Show Manager Panel 01 Object Reference)) -> (Rotator_CameraConfig:WBP Sparks Show Manager Rotator Object Reference)
    Index:int=Get ArrayElementLoc() -> (ArrayElementLoc:int=0)
   ) -> (then)
   n3 = Branch(execute=@n2.else, Condition="Equal (Integer)"(A=Get ArrayIndexLoc() -> (ArrayIndexLoc=0), B=1))
   > SetRotatorToSpecificIndex(
    self=Get Rotator_StageBackground!(self:WBP Sparks Show Manager Panel 01 Object Reference=Get Panel_Primary()) -> (Rotator_StageBackground:WBP Sparks Show Manager Rotator Object Reference)
    Index=Get ArrayElementLoc() -> (ArrayElementLoc=0)
   )
   n4 = Branch(execute=@n3.else, Condition="Equal (Integer)"(A=Get ArrayIndexLoc() -> (ArrayIndexLoc=0), B=2))
   > SetRotatorToSpecificIndex(
    self=Get Rotator_SongSection!(self:WBP Sparks Show Manager Panel 01 Object Reference=Get Panel_Primary()) -> (Rotator_SongSection:WBP Sparks Show Manager Rotator Object Reference)
    Index=Get ArrayElementLoc() -> (ArrayElementLoc=0)
   )
   n5 = Branch(execute=@n4.else, Condition="Equal (Integer)"(A=Get ArrayIndexLoc() -> (ArrayIndexLoc=0), B=3))
   > SetRotatorToSpecificIndex(
    self=Get Rotator_Color!(self:WBP Sparks Show Manager Panel 01 Object Reference=Get Panel_Primary()) -> (Rotator_Color:WBP Sparks Show Manager Rotator Object Reference)
    Index=Get ArrayElementLoc() -> (ArrayElementLoc=0)
   )
   n6 = Branch(execute=@n5.else, Condition="Equal (Integer)"(A=Get ArrayIndexLoc() -> (ArrayIndexLoc=0), B=4))
   > SetRotatorToSpecificIndex(
    self=Get Rotator_LocalAutoplay!(self:WBP Sparks Show Manager Panel 02 Object Reference=Get Panel_Secondary() -> (Panel_Secondary:WBP Sparks Show Manager Panel 02 Object Reference)) -> (Rotator_LocalAutoplay:WBP Sparks Show Manager Rotator Object Reference)
    Index=Get ArrayElementLoc() -> (ArrayElementLoc=0)
   )
   n7 = Branch(execute=@n6.else, Condition="Equal (Integer)"(A=Get ArrayIndexLoc() -> (ArrayIndexLoc=0), B=5))
   > SetRotatorToSpecificIndex(
    self=Get Rotator_GlobalAutoplay!(self:WBP Sparks Show Manager Panel 02 Object Reference=Get Panel_Secondary()) -> (Rotator_GlobalAutoplay:WBP Sparks Show Manager Rotator Object Reference)
    Index=Get ArrayElementLoc() -> (ArrayElementLoc=0)
   )
   n8 = Branch(execute=@n7.else, Condition="Equal (Integer)"(A=Get ArrayIndexLoc() -> (ArrayIndexLoc=0), B=6))
   > SetRotatorToSpecificIndex(
    self=Get Rotator_Aura!(self:WBP Sparks Show Manager Panel 02 Object Reference=Get Panel_Secondary()) -> (Rotator_Aura:WBP Sparks Show Manager Rotator Object Reference)
    Index=Get ArrayElementLoc() -> (ArrayElementLoc=0)
   )
   n9 = Branch(execute=@n8.else, Condition="Equal (Integer)"(A=Get ArrayIndexLoc() -> (ArrayIndexLoc=0), B=7))
   > SetRotatorToSpecificIndex(
    self=Get Rotator_CameraLogToScreen!(self:WBP Sparks Show Manager Panel 02 Object Reference=Get Panel_Secondary()) -> (Rotator_CameraLogToScreen:WBP Sparks Show Manager Rotator Object Reference)
    Index=Get ArrayElementLoc() -> (ArrayElementLoc=0)
   )
   n10 = Branch(execute=@n9.else, Condition="Equal (Integer)"(A=Get ArrayIndexLoc() -> (ArrayIndexLoc=0), B=8))
   > SetRotatorToSpecificIndex(
    self=Get Rotator_Vocals!(self:WBP Sparks Show Manager Panel 03 Object Reference=Get Panel_Tertiary() -> (Panel_Tertiary:WBP Sparks Show Manager Panel 03 Object Reference)) -> (Rotator_Vocals:WBP Sparks Show Manager Rotator Object Reference)
    Index=Get ArrayElementLoc() -> (ArrayElementLoc=0)
   )
   n11 = Branch(execute=@n10.else, Condition="Equal (Integer)"(A=Get ArrayIndexLoc() -> (ArrayIndexLoc=0), B=9))
   > SetRotatorToSpecificIndex(
    self=Get Rotator_Lead!(self:WBP Sparks Show Manager Panel 03 Object Reference=Get Panel_Tertiary()) -> (Rotator_Lead:WBP Sparks Show Manager Rotator Object Reference)
    Index=Get ArrayElementLoc() -> (ArrayElementLoc=0)
   )
   n12 = Branch(execute=@n11.else, Condition="Equal (Integer)"(A=Get ArrayIndexLoc() -> (ArrayIndexLoc=0), B=10))
   > SetRotatorToSpecificIndex(
    self=Get Rotator_Bass!(self:WBP Sparks Show Manager Panel 03 Object Reference=Get Panel_Tertiary()) -> (Rotator_Bass:WBP Sparks Show Manager Rotator Object Reference)
    Index=Get ArrayElementLoc() -> (ArrayElementLoc=0)
   )
   Branch(execute=@n12.else, Condition="Equal (Integer)"(A=Get ArrayIndexLoc() -> (ArrayIndexLoc=0), B=11))
   > SetRotatorToSpecificIndex(
    self=Get Rotator_Drums!(self:WBP Sparks Show Manager Panel 03 Object Reference=Get Panel_Tertiary()) -> (Rotator_Drums:WBP Sparks Show Manager Rotator Object Reference)
    Index=Get ArrayElementLoc() -> (ArrayElementLoc=0)
   )
  ]}
 EventGraph {nodes [
   "On Clicked (Button_Exit)"!() -> (OutputDelegate, then, Button:Common Button Base Object Reference) {is=title K2Node_ComponentBoundEvent}
   > ExecuteConsoleCommand!(execute, Command=Sparks.ToggleCaptureTool, SpecificPlayer) -> (then)
   "On Clicked (Button_Reset)"!() -> (OutputDelegate, then, Button:Common Button Base Object Reference) {is=title K2Node_ComponentBoundEvent}
   > n1 = Branch(Condition=GetActiveWidgetPanel!() -> (bFirstPanel:bool))
   > K2Node_Composite_2 = ResetPrimaryPanel() {is=title K2Node_Composite}
   K2Node_Composite_3 = ResetSecondaryPanel(execute=@n1.else) {is=title K2Node_Composite}
   "On Clicked (Button_Save)"!() -> (OutputDelegate, then, Button:Common Button Base Object Reference) {is=title K2Node_ComponentBoundEvent}
   > n2 = Create Widget!(
    execute:Exec
    Class:User Widget Class Reference=&/WBP_SparksShowManager_Prompt.WBP_SparksShowManager_Prompt_C
    OwningPlayer:Player Controller Object Reference
    bSave:bool=True
    bLoad:bool=False
   ) -> (then:Exec, ReturnValue:WBP Sparks Show Manager Prompt Object Reference)
   > n3 = Set SaveScreen(SaveScreen:WBP Sparks Show Manager Prompt Object Reference=@n2) -> (Output_Get:WBP Sparks Show Manager Prompt Object Reference)
   > AddToViewport!(execute, self:User Widget Object Reference=@n3, ZOrder:int=0) -> (then)
   n4 = Get Button_SaveLoad!(self:WBP Sparks Show Manager Prompt Object Reference=Get SaveScreen() -> (SaveScreen:WBP Sparks Show Manager Prompt Object Reference)) -> (Button_SaveLoad:WBP UIKit Button Loud Object Reference)
   > Unbind Event from On Clicked(self:Common Button Base Object Reference=@n4, Delegate=Create Event()) {is=title K2Node_RemoveDelegate}
   > Bind Event to On Clicked(self:Common Button Base Object Reference=@n4, Delegate=Create Event()) {is=title K2Node_AddDelegate}
   "On Clicked (Button_Load)"!() -> (OutputDelegate, then, Button:Common Button Base Object Reference) {is=title K2Node_ComponentBoundEvent}
   > n5 = Create Widget(Class=&WBP_SparksShowManager_Prompt_C, bSave=false, bLoad=true)
   > n6 = Set SaveScreen(SaveScreen=@n5)
   > AddToViewport(self=@n6, ZOrder=0)
   n7 = Get Button_SaveLoad(self=Get SaveScreen())
   > Unbind Event from On Clicked(self=@n7, Delegate=Create Event())
   > Bind Event to On Clicked(self=@n7, Delegate=Create Event())
   ClickButton_PrimaryTab!() -> (OutputDelegate, then, Button:Common Button Base Object Reference) {is=title K2Node_CustomEvent}
   > SetActiveWidgetIndex(self:Common Visibility Switcher Object Reference=Get Switcher() -> (Switcher:Common Visibility Switcher Object Reference), Index=0)
   > ToggleResetButton!(execute, self, bHide:bool=false) -> (then)
   > Set FocusTarget(FocusTarget:Widget Object Reference=Get Rotator_CameraConfig(self=Get Panel_Primary())) -> (Output_Get:Widget Object Reference)
   > SetControllerFocus()
   ClickButton_SecondaryTab!() -> (OutputDelegate, then, Button:Common Button Base Object Reference) {is=title K2Node_CustomEvent}
   > SetActiveWidgetIndex(self=Get Switcher(), Index=1)
   > ToggleResetButton(bHide=false)
   > Set FocusTarget(FocusTarget=Get Rotator_LocalAutoplay(self=Get Panel_Secondary()))
   > SetControllerFocus()
   ClickButton_TertiaryTab!() -> (OutputDelegate, then, Button:Common Button Base Object Reference) {is=title K2Node_CustomEvent}
   > SetActiveWidgetIndex(self=Get Switcher(), Index=2)
   > ToggleResetButton(bHide=true)
   > Set FocusTarget(FocusTarget=Get Rotator_Vocals(self=Get Panel_Tertiary()))
   > SetControllerFocus()
   n8 = SetControllerFocus() {is=title K2Node_CustomEvent}
   n9 = Get CommonInputSubsystem()
   > n10 = Is Valid(InputObject=@n9)
   n11 = GetCurrentInputType(self=@n9)
   > n12 = Switch on ECommonInputType!(execute, Selection:ECommonInputType Enum=@n11) -> (MouseAndKeyboard:Exec, Gamepad:Exec, Touch:Exec, Count:Exec) {is=title K2Node_SwitchEnum}
   n13 = SetFocus(execute=[@n12.Gamepad, @n10.Is Not Valid], self=Get FocusTarget() -> (FocusTarget:Widget Object Reference))
   SetVisibility(
    execute=[@n13.then, @n12.MouseAndKeyboard]
    self=[
     Get Button_Save() -> (Button_Save:WBP UIKit Button Regular Object Reference)
     Get Button_Load() -> (Button_Load:WBP UIKit Button Regular Object Reference)
    ]
    InVisibility=Select!(
     MouseAndKeyboard:ESlateVisibility Enum=SelfHitTestInvisible
     Gamepad:ESlateVisibility Enum=Collapsed
     Touch:ESlateVisibility Enum=SelfHitTestInvisible
     Count:ESlateVisibility Enum=Collapsed
     Index:ECommonInputType Enum=@n11
    ) -> (ReturnValue)
   )
   Event Construct()
   > n14 = GetActorOfClass!(
    execute
    ActorClass:Actor Class Reference=/PilgrimQuickplay/StageManager/StageShowManager/BP_PilgrimStageShowManager.BP_PilgrimStageShowManager_C
   ) -> (then, ReturnValue:BP Pilgrim Stage Show Manager Object Reference)
   > Set PilgrimManager(PilgrimManager:BP Pilgrim Stage Show Manager Object Reference=@n14) -> (Output_Get:BP Pilgrim Stage Show Manager Object Reference)
   > Set FocusTarget(FocusTarget=Get Rotator_CameraConfig(self=Get Panel_Primary()))
   > SetControllerFocus() {is=title K2Node_CallFunction}
   > Bind Event to On Keybinds Changed(
    self:Fort Global UIContext Object Reference=Get FortGlobalUIContext() -> (ReturnValue:Fort Global UIContext Object Reference) {is=title K2Node_GetSubsystem}
    Delegate=@n8
   ) {is=title K2Node_AddDelegate}
   > K2Node_Composite_1 = BindClickTabEvents!(execute) -> (then:Exec) {is=title K2Node_Composite}
   > SetNavRules()
  ]}
 BindClickTabEvents {
  path=EventGraph.K2Node_Composite_1.BindClickTabEvents
  nodes [
   Inputs()
   > Bind Event to On Clicked(
    self=Get Button_Tab1!(self:WBP Sparks Show Manager Tabs Object Reference=Get Panel_TabButtons() -> (Panel_TabButtons:WBP Sparks Show Manager Tabs Object Reference)) -> (Button_Tab1:WBP UIKit Button Selectable Object Reference)
    Delegate=Create Event()
   )
   > Bind Event to On Clicked(
    self=Get Button_Tab2!(self:WBP Sparks Show Manager Tabs Object Reference=Get Panel_TabButtons()) -> (Button_Tab2:WBP UIKit Button Selectable Object Reference)
    Delegate=Create Event()
   )
   > Bind Event to On Clicked(
    self=Get Button_Tab3!(self:WBP Sparks Show Manager Tabs Object Reference=Get Panel_TabButtons()) -> (Button_Tab3:WBP UIKit Button Selectable Object Reference)
    Delegate=Create Event()
   )
   > Outputs!(then:Exec)
  ]
 }
 ResetPrimaryPanel {
  path=EventGraph.K2Node_Composite_2.ResetPrimaryPanel
  nodes [
   Inputs()
   > SyncToCurrentState(self:BP Pilgrim Stage Show Manager Object Reference=Get PilgrimManager() -> (PilgrimManager:BP Pilgrim Stage Show Manager Object Reference))
   n1 = Get Panel_Primary()
   > ResetToDefault(self:WBP Sparks Show Manager Rotator Object Reference=[Get Rotator_CameraConfig(self=@n1), Get Rotator_StageBackground(self=@n1), Get Rotator_SongSection(self=@n1), Get Rotator_Color(self=@n1)])
   Outputs!()
  ]
 }
 ResetSecondaryPanel {
  path=EventGraph.K2Node_Composite_3.ResetSecondaryPanel
  nodes [
   Inputs()
   > ExecuteConsoleCommand(Command=Cheat Pilgrim.Autoplay 0, SpecificPlayer=GetOwningPlayer())
   > ExecuteConsoleCommand(Command=Pilgrim.Autoplay 0, SpecificPlayer=GetOwningPlayer())
   > ExecuteConsoleCommand(Command=Sparks.ForceAura 1, SpecificPlayer=GetOwningPlayer())
   > ExecuteConsoleCommand(Command=Sparks.CameraLogtoScreen 0, SpecificPlayer=GetOwningPlayer())
   n1 = Get Panel_Secondary()
   > ResetToDefault(self=[
     Get Rotator_LocalAutoplay(self=@n1)
     Get Rotator_GlobalAutoplay(self=@n1)
     Get Rotator_Aura(self=@n1)
     Get Rotator_CameraLogToScreen(self=@n1)
    ])
   Outputs()
  ]
 }
 GetActiveWidgetPanel {nodes [
   GetActiveWidgetPanel() {is=title K2Node_FunctionEntry}
   n1 = Get Switcher()
   > n2 = Branch(Condition="Equal (Integer)"(A=GetActiveWidgetIndex!(self:Common Visibility Switcher Object Reference=@n1) -> (ReturnValue:int=0)))
   > Return Node!(execute, bFirstPanel:bool=true)
   n3 = Get Switcher()
   Branch(execute=@n2.else, Condition="Equal (Integer)"(A=GetActiveWidgetIndex(self=@n3) -> (ReturnValue=0), B=1))
   > Return Node(bFirstPanel=false)
  ]}
 LoadSaveFile {nodes [
   LoadSaveFile!() -> (then, Button:Common Button Base Object Reference) {is=title K2Node_FunctionEntry}
   n1 = Get SaveScreen()
   n2 = "To String (Text)"(InText="GetText (Editable Text)"!(self:Editable Text Object Reference=Get EditableText_FileName!(self:WBP Sparks Show Manager Prompt Object Reference=@n1) -> (EditableText_FileName:Editable Text Object Reference)) -> (ReturnValue:Text))
   > n3 = DoesSaveGameExist!(execute, SlotName:String=@n2, UserIndex:int=0) -> (then, ReturnValue:bool)
   > n4 = Branch(Condition=@n3)
   > n5 = LoadGameFromSlot!(execute, SlotName:String=@n2, UserIndex:int=0) -> (then, ReturnValue:Save Game Object Reference)
   > n6 = Cast To BP_SparksShowManager_SaveGame!(execute, Object=@n5) -> (then, CastFailed, AsBP Sparks Show Manager Save Game:BP Sparks Show Manager Save Game Object Reference) {is=title K2Node_DynamicCast}
   > n7 = LoadSaveData!(execute, self:BP Sparks Show Manager Save Game Object Reference=@n6) -> (
    then
    SaveData_Config_2_#1:int
    SaveData_StageBackground_10_#2:int
    SaveData_SongSection_11_#3:int
    SaveData_ColorPalette_12_#4:int
    SaveData_LocalAutoplay_13_#5:int
    SaveData_GlobalAutoplay_14_#6:int
    SaveData_Aura_15_#7:int
    SaveData_CameraLogToScreen_16_#8:int
    SaveData_VocalPosition_21_#9:int
    SaveData_LeadPosition_22_#10:int
    SaveData_BassPosition_23_#11:int
    SaveData_DrumPosition_24_#12:int
   )
   > n8 = For Each Loop(Array:Array of Integers=Make Array!(
     "[0]":int=@n7
     "[1]":int=@n7.SaveData_StageBackground_10_#2
     "[2]":int=@n7.SaveData_SongSection_11_#3
     "[3]":int=@n7.SaveData_ColorPalette_12_#4
     "[4]":int=@n7.SaveData_LocalAutoplay_13_#5
     "[5]":int=@n7.SaveData_GlobalAutoplay_14_#6
     "[6]":int=@n7.SaveData_Aura_15_#7
     "[7]":int=@n7.SaveData_CameraLogToScreen_16_#8
     "[8]":int=@n7.SaveData_VocalPosition_21_#9
     "[9]":int=@n7.SaveData_LeadPosition_22_#10
     "[10]":int=@n7.SaveData_BassPosition_23_#11
     "[11]":int=@n7.SaveData_DrumPosition_24_#12
    ) -> (Array:Array of Integers)) -> (Array Element:int)
   > ApplyRotatorPositionsFromSave!(execute, self, ArrayElement:int=@n8, ArrayIndex:int=@n8.Array Index) -> (then) {is=title K2Node_CallFunction}
   RunEntirePrimaryPanel(execute=@n8.Completed, self:WBP Sparks Show Manager Panel 01 Object Reference=Get Panel_Primary())
   > RunEntireSecondaryPanel(self:WBP Sparks Show Manager Panel 02 Object Reference=Get Panel_Secondary())
   > RunEntireTertiaryPanel(self:WBP Sparks Show Manager Panel 03 Object Reference=Get Panel_Tertiary())
   > CloseMenu(self:WBP Sparks Show Manager Prompt Object Reference=Get SaveScreen())
   PrintString!(
    execute=@n4.else
    InString="No save file with that name exists "
    bPrintToScreen
    bPrintToLog
    TextColor=(R=1,G=0.0099,B=0.0398,A=1)
    Duration=3
    Key
   ) -> (then)
  ]}
 SaveData {nodes [
   SaveData!() -> (then, Button:Common Button Base Object Reference) {is=title K2Node_FunctionEntry}
   n1 = Get EditableText_FileName(self=Get SaveScreen())
   > Set SaveFileName(SaveFileName:String="To String (Text)"(InText="GetText (Editable Text)"(self=@n1))) -> (Output_Get:String)
   > n2 = CreateSaveGameObject!(execute, SaveGameClass:Save Game Class Reference=&/BP_SparksShowManager_SaveGame.BP_SparksShowManager_SaveGame_C) -> (then, ReturnValue:BP Sparks Show Manager Save Game Object Reference)
   K2Node_Composite_3 = AllPanelSettings!() -> (
    Config:int
    StageBackground:int
    SongSection:int
    ColorPalette:int
    LocalAutoplay:int
    GlobalAutoplay:int
    Aura:int
    CameraLogToScreen:int
    VocalPosition:int
    LeadPosition:int
    BassPosition:int
    DrumPosition:int
   ) {is=title K2Node_Composite}
   > CollectSaveData!(
    execute
    self:BP Sparks Show Manager Save Game Object Reference=@n2
    ConfigIndex:int=@K2Node_Composite_3
    StageBackgroundIndex:int=@K2Node_Composite_3.StageBackground
    SongSectionIndex:int=@K2Node_Composite_3.SongSection
    ColorPaletteIndex:int=@K2Node_Composite_3.ColorPalette
    LocalAutoplayIndex:int=@K2Node_Composite_3.LocalAutoplay
    GlobalAutoplayIndex:int=@K2Node_Composite_3.GlobalAutoplay
    AuraIndex:int=@K2Node_Composite_3.Aura
    CameraLogToScreenIndex:int=@K2Node_Composite_3.CameraLogToScreen
    VocalPositionIndex:int=@K2Node_Composite_3.VocalPosition
    LeadPositionIndex:int=@K2Node_Composite_3.LeadPosition
    BassPositionIndex:int=@K2Node_Composite_3.BassPosition
    DrumsPositionIndex:int=@K2Node_Composite_3.DrumPosition
   ) -> (then)
   > n3 = SaveGameToSlot!(execute, SaveGameObject:Save Game Object Reference=@n2, SlotName:String=Get SaveFileName() -> (SaveFileName:String), UserIndex:int=0) -> (then, ReturnValue:bool)
   > Branch(Condition=@n3)
   n4 = "GetText (Editable Text)"(self=Get EditableText_FileName(self=Get SaveScreen()))
   > PrintString(
    InString=Append!(A="SAVE FILE \"", B="To String (Text)"(InText=@n4), C:String="\" SUCCESSFULLY CREATED") -> (ReturnValue)
    TextColor=(R=0,G=0.6042,B=0.074,A=1)
    Duration=3
   )
   > CloseMenu(self=Get SaveScreen())
  ]}
 AllPanelSettings {
  path=SaveData.K2Node_Composite_3.AllPanelSettings
  nodes [
   Inputs!()
   K2Node_Composite_0 = PrimaryPanelSettings_2!() -> (Config:int, StageBackground:int, SongSection:int, ColorPalette:int) {is=title K2Node_Composite}
   K2Node_Composite_1 = SecondaryPanelSettings_2!() -> (LocalAutoplay:int, GlobalAutoplay:int, Aura:int, CameraLogToScreen:int) {is=title K2Node_Composite}
   K2Node_Composite_2 = TertiaryPanelSettings_2!() -> (VocalPosition:int, LeadPosition:int, BassPosition:int, DrumPosition:int) {is=title K2Node_Composite}
   Outputs!(
    Config:int=@K2Node_Composite_0
    StageBackground:int=@K2Node_Composite_0.StageBackground
    SongSection:int=@K2Node_Composite_0.SongSection
    ColorPalette:int=@K2Node_Composite_0.ColorPalette
    LocalAutoplay:int=@K2Node_Composite_1
    GlobalAutoplay:int=@K2Node_Composite_1.GlobalAutoplay
    Aura:int=@K2Node_Composite_1.Aura
    CameraLogToScreen:int=@K2Node_Composite_1.CameraLogToScreen
    VocalPosition:int=@K2Node_Composite_2
    LeadPosition:int=@K2Node_Composite_2.LeadPosition
    BassPosition:int=@K2Node_Composite_2.BassPosition
    DrumPosition:int=@K2Node_Composite_2.DrumPosition
   )
  ]
 }
 PrimaryPanelSettings_2 {
  path=SaveData.K2Node_Composite_3.AllPanelSettings.K2Node_Composite_0.PrimaryPanelSettings_2
  nodes [
   Inputs()
   n1 = Get Panel_Primary()
   Outputs!(
    Config=Get RotatorIndex!(self:WBP Sparks Show Manager Rotator Object Reference=Get Rotator_CameraConfig(self=@n1)) -> (RotatorIndex:int=0)
    StageBackground=Get RotatorIndex(self=Get Rotator_StageBackground(self=@n1)) -> (RotatorIndex=0)
    SongSection=Get RotatorIndex(self=Get Rotator_SongSection(self=@n1)) -> (RotatorIndex=0)
    ColorPalette=Get RotatorIndex(self=Get Rotator_Color(self=@n1)) -> (RotatorIndex=0)
   )
  ]
 }
 SecondaryPanelSettings_2 {
  path=SaveData.K2Node_Composite_3.AllPanelSettings.K2Node_Composite_1.SecondaryPanelSettings_2
  nodes [
   Inputs()
   n1 = Get Panel_Secondary()
   Outputs!(
    LocalAutoplay:int=Get RotatorIndex(self=Get Rotator_LocalAutoplay(self=@n1)) -> (RotatorIndex=0)
    GlobalAutoplay:int=Get RotatorIndex(self=Get Rotator_GlobalAutoplay(self=@n1)) -> (RotatorIndex=0)
    Aura:int=Get RotatorIndex(self=Get Rotator_Aura(self=@n1)) -> (RotatorIndex=0)
    CameraLogToScreen:int=Get RotatorIndex(self=Get Rotator_CameraLogToScreen(self=@n1)) -> (RotatorIndex=0)
   )
  ]
 }
 TertiaryPanelSettings_2 {
  path=SaveData.K2Node_Composite_3.AllPanelSettings.K2Node_Composite_2.TertiaryPanelSettings_2
  nodes [
   Inputs()
   n1 = Get Panel_Tertiary()
   Outputs!(
    VocalPosition:int=Get RotatorIndex(self=Get Rotator_Vocals(self=@n1)) -> (RotatorIndex=0)
    LeadPosition:int=Get RotatorIndex(self=Get Rotator_Lead(self=@n1)) -> (RotatorIndex=0)
    BassPosition:int=Get RotatorIndex(self=Get Rotator_Bass(self=@n1)) -> (RotatorIndex=0)
    DrumPosition:int=Get RotatorIndex(self=Get Rotator_Drums(self=@n1)) -> (RotatorIndex=0)
   )
  ]
 }
 SetNavRules {nodes [
   SetNavRules() {is=title K2Node_FunctionEntry}
   > SetNavigationRuleExplicit!(
    execute
    self:Widget Object Reference=Get Button_Reset() -> (Button_Reset:WBP UIKit Button Regular Object Reference)
    Direction:EUINavigation Enum=Left
    InWidget:Widget Object Reference=Get Button_Exit() -> (Button_Exit:WBP UIKit Button Regular Object Reference)
   ) -> (then)
   > SetNavigationRuleExplicit(self=Get Button_Exit(), Direction=Right, InWidget=Get Button_Reset())
   > SetNavigationRuleExplicit(self=Get Button_Exit(), Direction=Down, InWidget=Get FocusTarget())
   > SetNavigationRuleExplicit(self=Get Button_Reset(), Direction=Down, InWidget=Get FocusTarget())
  ]}
 ToggleResetButton {nodes [
   n1 = ToggleResetButton!() -> (then, bHide:bool=~) {is=title K2Node_FunctionEntry}
   > n2 = Branch(Condition=@n1)
   > SetVisibility(self=Get Button_Reset(), InVisibility=Collapsed)
   SetVisibility(execute=@n2.else, self=Get Button_Reset(), InVisibility=Visible)
  ]}
]
variables [
 FocusTarget {type=Widget, default=null}
 PilgrimManager {type=BP_PilgrimStageShowManager_C, default=null}
 SaveFileName {type=string, default=Temp Name}
 SaveScreen {type=WBP_SparksShowManager_Prompt_C, default=null}
]
widgets {
 info {inheritedWidgetCount=0, parentClass=&UserWidget, rootWidgetClass=&CanvasPanel, widgetCount=17}
 widgets [
  CanvasPanel {widgetClassPath=&CanvasPanel, widget=CanvasPanel}
  Backing {
   widgetClassPath=&Overlay
   parent=CanvasPanel
   widget=Backing
   slot {class=CanvasPanelSlot, properties {layoutData {alignment=<1 0.5>, anchors {maximum=<1 0.5>, minimum=<1 0.5>}, offsets=<ltrb -25 0 650 700>}}}
  }
  Image_Background {
   widgetClassPath=&Border
   parent=Backing
   widget=Image_Background
   properties {
    background {
     bIsDynamicallyLoaded=false
     drawAs=RoundedBox
     imageType=NoImage
     mirroring=NoMirror
     resourceName=None
     resourceObject=null
     tiling=NoTile
     imageSize=<32 32>
     margin=<ltrb 0 0 0 0>
     outlineSettings {
      bUseBrushTransparency=false
      roundingType=FixedRadius
      width=0
      color {colorUseRule=UseColor_Specified, specifiedColor=<rgba 0 0 0 0>}
      cornerRadii=<25 25 25 25>
     }
     tintColor {colorUseRule=UseColor_Specified, specifiedColor=<rgba 0.0409 0.0168 0.1683 0.92>}
     uVRegion {bIsValid=false, max=<0 0>, min=<0 0>}
    }
    padding=<ltrb 14 14 14 14>
   }
   slot {class=OverlaySlot, properties {horizontalAlignment=HAlign_Fill, verticalAlignment=VAlign_Fill}}
  }
  Blur_Background {
   widgetClassPath=&BackgroundBlur
   parent=Backing
   widget=Blur_Background
   properties {blurStrength=0.6}
   slot {class=OverlaySlot, properties {horizontalAlignment=HAlign_Fill, verticalAlignment=VAlign_Fill}}
  }
  MasterBox {
   widgetClassPath=&VerticalBox
   parent=Backing
   widget=MasterBox
   slot {class=OverlaySlot, properties {horizontalAlignment=HAlign_Fill, verticalAlignment=VAlign_Fill}}
  }
  Heading {
   widgetClassPath=&VerticalBox
   parent=MasterBox
   widget=Heading
   slot {class=VerticalBoxSlot, properties {verticalAlignment=VAlign_Top, padding=<ltrb 0 25 0 0>}}
  }
  MenuTitle {
   widgetClassPath=&CommonTextBlock
   parent=Heading
   widget=MenuTitle
   properties {
    justification=Center
    text=Show Manager
    textTransformPolicy=ToUpper
    visibility=HitTestInvisible
    colorAndOpacity {colorUseRule=UseColor_Specified, specifiedColor=<rgba 0.5 0.8333 1 1>}
    font {
     bForceMonospaced=false
     bMaterialIsStencil=false
     fontMaterial=null
     fontObject=&HeadingNow
     letterSpacing=0
     monospacedWidth=1
     size=33.75
     skewAmount=0
     typefaceFontName=76 Bold Italic
     outlineSettings {
      bApplyOutlineToDropShadows=false
      bMiteredCorners=false
      bSeparateFillAlpha=false
      outlineMaterial=null
      outlineSize=0
      outlineColor=<rgba 0 0 0 1>
     }
    }
    shadowOffset=<0 0>
   }
   slot {class=VerticalBoxSlot, properties {horizontalAlignment=HAlign_Center, verticalAlignment=VAlign_Center}}
  }
  Panel_TabButtons {
   widgetClassPath=&/WBP_SparksShowManager_Tabs
   bIsVariable=true
   parent=Heading
   widget=Panel_TabButtons
   slot {class=VerticalBoxSlot, properties {horizontalAlignment=HAlign_Center, padding=<ltrb 0 25 0 0>}}
  }
  Switcher {
   widgetClassPath=&CommonVisibilitySwitcher
   bIsVariable=true
   parent=MasterBox
   widget=Switcher
   slot {class=VerticalBoxSlot, properties {padding=<ltrb 0 45 0 0>}}
  }
  Panel_Primary {
   widgetClassPath=&/WBP_SparksShowManager_Panel_01
   bIsVariable=true
   parent=Switcher
   widget=Panel_Primary
   slot {class=CommonVisibilitySwitcherSlot}
  }
  Panel_Secondary {
   widgetClassPath=&/WBP_SparksShowManager_Panel_02
   bIsVariable=true
   parent=Switcher
   widget=Panel_Secondary
   slot {class=CommonVisibilitySwitcherSlot}
  }
  Panel_Tertiary {
   widgetClassPath=&/WBP_SparksShowManager_Panel_03
   bIsVariable=true
   parent=Switcher
   widget=Panel_Tertiary
   slot {class=CommonVisibilitySwitcherSlot}
  }
  Box_Buttons {
   widgetClassPath=&HorizontalBox
   parent=MasterBox
   widget=Box_Buttons
   slot {
    class=VerticalBoxSlot
    properties {horizontalAlignment=HAlign_Right, verticalAlignment=VAlign_Bottom, padding=<ltrb 0 0 15 20>, size {sizeRule=Fill, value=1}}
   }
  }
  Button_Save {
   widgetClassPath=&WBP_UIKit_Button_Regular
   bIsVariable=true
   parent=Box_Buttons
   widget=Button_Save
   properties {buttonSize=Small, buttonType=Quiet, text=Save}
   slot {class=HorizontalBoxSlot, properties {padding=<ltrb 0 0 15 0>}}
  }
  Button_Load {
   widgetClassPath=&WBP_UIKit_Button_Regular
   bIsVariable=true
   parent=Box_Buttons
   widget=Button_Load
   properties {buttonSize=Small, buttonType=Quiet, text=Load}
   slot {class=HorizontalBoxSlot, properties {padding=<ltrb 0 0 15 0>}}
  }
  Button_Reset {
   widgetClassPath=&WBP_UIKit_Button_Regular
   bIsVariable=true
   parent=Box_Buttons
   widget=Button_Reset
   properties {bHideInputActionWithKeyboard=true, bRequiresHold=true, buttonSize=Small, buttonType=Quiet, text=Reset}
   slot {class=HorizontalBoxSlot, properties {padding=<ltrb 0 0 15 0>}}
  }
  Button_Exit {
   widgetClassPath=&WBP_UIKit_Button_Regular
   bIsVariable=true
   parent=Box_Buttons
   widget=Button_Exit
   properties {bHideInputActionWithKeyboard=true, buttonSize=Small, buttonType=Quiet, text=Exit}
   slot {class=HorizontalBoxSlot}
  }
 ]
}
